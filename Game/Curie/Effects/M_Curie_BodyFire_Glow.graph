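# Additive unlit Niagara mesh-particle material that draws the Curie body fire glow over the scene behind it.
# An inverted fresnel on scene world normals, constant on mobile and forward, shapes a base glow plus OnFire-driven panning noise and sine pulse.
# Sphere mask, behind-translucency distance, timestamp and depth fade attenuate the tinted result.
# IsLocalPlayer blends between height-above-pivot with custom-depth occlusion and an inverted custom-depth mask.
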Material /Game/Curie/Effects/M_Curie_BodyFire_Glow
graph {
 properties {
  bEnableResponsiveAA=true
  bUsedWithNiagaraMeshParticles=true
  blendMode=BLEND_Additive
  pixelDepthOffsetMode=PDOM_Legacy
  shadingModel=MSM_Unlit
  usedShadingModels=Unlit
 }
 nodes [
  n1 = Constant() {r=0.5}
  n2 = ScalarParameter() {defaultValue=0.02, parameterName=BaseGlow}
  n3 = SceneTexture() {sceneTextureId=PPI_WorldNormal}
  n4 = ScalarParameter() {defaultValue=0.025, parameterName=MobileOpacity}
  n5 = ShadingPathSwitch(
   Default=MaterialFunctionCall(
    Exponent=ScalarParameter() {defaultValue=6, parameterName=FresnelExponent}
    Vertex / Pixel Normal=ComponentMask(None=@n3.Color) {b=true, g=true, r=true}
   ) {materialFunction=&MF_Fresnel_FX}.Fresnel Inverted
   Forward=@n4
   Mobile=@n4
  )
  n6 = LinearInterpolate(A=ScalarParameter() {defaultValue=0.4, parameterName=WanringGlowFresnelInner}, Alpha=@n5) {constA=0.5}
  n7 = TextureCoordinate()
  n8 = ParticleRandom()
  n9 = VectorParameter() {parameterName=FlamesScalePan, defaultValue=<rgba 0.3 0.3 0 0.45>}
  n10 = Panner(Coordinate=Multiply(A=Add(A=@n7, B=@n8), B=AppendVector(A=@n9.R, B=@n9.R)), Speed=AppendVector(A=@n9.B, B=@n9.B)) {bFractionalPart=true}
  n11 = ConstantBiasScale(None=Sine(None=Time()) {period=0.3}) {bias=1.1}
  n12 = Add(
   A=Multiply(
    A=TextureSample(UVs=@n10) {samplerType=SAMPLERTYPE_LinearColor, texture=/Game/Global/Textures/Default/Noise/T_Curie_NoisePack}.G
    B=ScalarParameter() {defaultValue=0.1, parameterName=FlamesScale}
   )
   B=Multiply(A=@n5, B=Saturate(None=@n11))
  )
  n13 = DynamicParameter() {paramNames [OnFire, IsLocalPlayer, UseTimestampFade, Param4]}
  n14 = Multiply(
   A=SphereMask(A=TextureCoordinate(), B=@n1, Radius=@n1, Hardness=Constant() {r=0.5})
   B=Add(A=Multiply(A=@n2, B=@n6), B=Multiply(A=@n12, B=@n13.OnFire))
  )
  n15 = MaterialFunctionCall!(
   Distance Bias=ScalarParameter() {defaultValue=50, parameterName=CameraOffsetBias}
   Distance Divisor=ScalarParameter() {defaultValue=30, parameterName=DistanceDivisor}
   Scene Depth
   Pixel Depth
  ) {materialFunction=/Game/Effects/Fort_Effects/Materials/_Functions/MF_DistanceBehindTranslucency_01}
  n16 = DynamicParameter() {parameterIndex=1, paramNames [TimeStampA, TransitionLengthIn, TimeStampB, TransitionLengthOut]}
  n17 = StaticBool()
  n18 = Multiply(
   A=Multiply(A=VectorParameter() {parameterName=Color, defaultValue=<rgba 5 0.3698 0.0213 1>}.RGB, B=Multiply(A=@n14, B=@n15.Inverted))
   B=VertexInterpolator(VS=MaterialFunctionCall(
     TimestampA=@n16.TimeStampA
     TransitionLength_In=@n16.TimeStampA
     TimestampB=@n16.TimeStampA
     TransitionLength_Out=@n16.TimeStampA
     Transition In Only?=@n17
    ) {materialFunction=&MF_Timestamp_InOut_01}.Linear).PS
  )
  n19 = DepthFade(FadeDistance=ScalarParameter() {defaultValue=10, parameterName=FadeDistance})
  n20 = ComponentMask(None=Subtract(
    A=MaterialFunctionCall() {materialFunction=&WorldPositionBehindTranslucency}.Result
    B=MaterialFunctionCall() {materialFunction=&ObjectPivotPoint}.Object Pivot Location
   )) {b=true}
  n21 = ScalarParameter() {defaultValue=50, parameterName=ZMaskDivisor}
  n22 = ComponentMask(None=SceneTexture() {sceneTextureId=PPI_CustomDepth}.Color) {r=true}
  n23 = Saturate(None=Divide(A=@n22) {constB=10000})
  n24 = SceneDepth()
  n25 = Constant()
  n26 = Constant() {r=1}
  n27 = Multiply(A=Saturate(None=Divide(A=@n20, B=@n21)), B=LinearInterpolate(A=@n23, Alpha=If(A=@n24.R, B=@n22, "A > B"=@n25, "A < B"=@n26)))
  n28 = OneMinus(None=@n23)
  n29 = DynamicParameter() {paramNames [OnFire, IsLocalPlayer, UseTimestampFade, Param4]}
  MaterialFunctionCall(Emissive=Multiply(A=Multiply(A=@n18, B=@n19), B=LinearInterpolate(A=@n27, B=@n28, Alpha=@n29.IsLocalPlayer) {constA=1})) {materialFunction=&MF_FX_EmissiveAdjustments}
 ]
}
