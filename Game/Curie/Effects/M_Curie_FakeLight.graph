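# Additive unlit Niagara mesh-particle material that fakes a point light on the surfaces behind it.
# GBuffer normal and base color shade a sphere mask around the particle, tinted by particle color.
# Vertex-interpolated distance fade, dual-sine flicker and lifetime sweep mask modulate radius or brightness.
# Offset pushes the mesh toward the camera by light radius, collapsing it when faded; Forward and Mobile paths collapse it and emit nothing.
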Material /Game/Curie/Effects/M_Curie_FakeLight
graph {
 properties {
  bEnableResponsiveAA=true
  bUsedWithNiagaraMeshParticles=true
  blendMode=BLEND_Additive
  pixelDepthOffsetMode=PDOM_Legacy
  shadingModel=MSM_Unlit
  usedShadingModels=Unlit
 }
 nodes [
  n1 = ParticleColor()
  n2 = ParticleRadius()
  n3 = MaterialFunctionCall(
   Position=ParticlePositionWS()
   Fade_Start_Distance=Add(A=ScalarParameter() {defaultValue=500, parameterName=FadeStartDistance}, B=@n2)
   Fade_End_Distance=Add(A=ScalarParameter() {defaultValue=350, parameterName=FadeEndDistance}, B=@n2)
  ) {materialFunction=&MF_Distance_SyncedFade_FX}
  n4 = DynamicParameter() {desc="##1", parameterIndex=1, defaultValue=<rgba 0 0 0 1>, paramNames [LightRadius, FlickerSpeed, FlickerScale, LightAlphaScale]}
  n5 = Multiply(A=@n4.FlickerSpeed, B=Time()) {constA=1}
  n6 = DynamicParameter() {desc="##1", parameterIndex=1, defaultValue=<rgba 0 0 0 1>, paramNames [LightRadius, FlickerSpeed, FlickerScale, LightAlphaScale]}
  n7 = Multiply(A=Add(A=Sine(None=@n5) {period=0.5651}, B=Sine(None=@n5) {period=0.442}), B=@n6.FlickerScale)
  n8 = DynamicParameter() {desc="##2", parameterIndex=2, defaultValue=<rgba 0 0 0 175>, paramNames [X, Y, Z, SphericalRadius]}
  n9 = SphereMask(
   A=MaterialFunctionCall(R=@n8.X, G=@n8.X, B=@n8.X) {materialFunction=&AppendMany}.RGB
   B=ParticlePositionWS()
   Radius=@n8.SphericalRadius
   Hardness=Constant()
  )
  n10 = ParticleRelativeTime()
  n11 = SphereMask(
   A=@n9
   B=ConstantBiasScale(None=OneMinus(None=@n10)) {bias=-0.5, scale=2}
   Radius=ScalarParameter() {defaultValue=0.3, parameterName=SweepRadius}
   Hardness=Constant() {r=0.2}
  )
  n12 = OneMinus(None=Multiply(A=Ceil(None=@n9), B=OneMinus(None=@n11)))
  n13 = VertexInterpolator(VS=MaterialFunctionCall(R=@n3.Result, G=Add(A=@n7), B=@n12) {materialFunction=&AppendMany}.RGB)
  n14 = Normalize(VectorInput=Subtract(
    A=ParticlePositionWS()
    B=MaterialFunctionCall() {materialFunction=&WorldPositionBehindTranslucency}.Result
   ))
  n15 = ComponentMask(None=SceneTexture() {sceneTextureId=PPI_WorldNormal}.Color) {b=true, g=true, r=true}
  n16 = ParticlePositionWS()
  n17 = MaterialFunctionCall() {materialFunction=&WorldPositionBehindTranslucency}
  n18 = DynamicParameter() {desc="##1", parameterIndex=1, defaultValue=<rgba 0 0 0 1>, paramNames [LightRadius, FlickerSpeed, FlickerScale, LightAlphaScale]}
  n19 = DynamicParameter() {desc="##3", parameterIndex=3, paramNames [isIgnition,HazeScale,Param3,LightScale]}
  n20 = Multiply(
   A=Multiply(A=Multiply(A=@n18.LightRadius, B=@n19.LightScale), B=ComponentMask(None=@n13.PS) {b=true})
   B=ComponentMask(None=@n13.PS) {r=true}
  )
  n21 = Constant() {r=0.1}
  n22 = Multiply(
   A=Saturate(None=DotProduct(A=@n14, B=@n15))
   B=Power(
    Base=SphereMask(A=@n16, B=@n17.Result, Radius=@n20, Hardness=@n21)
    Exp=ScalarParameter() {defaultValue=4, parameterName=RadiusFalloffExponent}
   )
  )
  n23 = LinearInterpolate(B=ComponentMask(None=SceneTexture() {sceneTextureId=PPI_StoredBaseColor}.Color) {b=true, g=true, r=true}) {constA=1, constAlpha=0.4}
  n24 = Multiply(A=ComponentMask(None=@n13.PS) {g=true}, B=Multiply(A=ComponentMask(None=@n13.PS) {r=true}, B=Multiply(A=@n22, B=@n23)))
  n25 = Constant()
  MaterialFunctionCall(Emissive=ShadingPathSwitch(Default=Multiply(A=@n1.RGB, B=@n24), Forward=@n25, Mobile=@n25)) {materialFunction=&MF_FX_EmissiveAdjustments}
  n26 = DynamicParameter() {desc="##1", parameterIndex=1, defaultValue=<rgba 0 0 0 1>, paramNames [LightRadius, FlickerSpeed, FlickerScale, LightAlphaScale]}
  Multiply(A=@n1.A, B=@n26.LightAlphaScale)
  n27 = Subtract(A=ParticlePositionWS(), B=WorldPosition().XYZ)
  n28 = Normalize(VectorInput=Subtract(A=CameraPositionWS(), B=ParticlePositionWS()))
  n29 = DynamicParameter() {desc="##1", parameterIndex=1, defaultValue=<rgba 0 0 0 1>, paramNames [LightRadius, FlickerSpeed, FlickerScale, LightAlphaScale]}
  n30 = Min(A=@n3.Result, B=DynamicParameter() {desc="##3", parameterIndex=3, paramNames [isIgnition,HazeScale,Param3,LightScale]}.LightScale)
  n31 = Multiply(A=@n30, B=@n12)
  n32 = DynamicParameter() {desc="##3", parameterIndex=3, paramNames [isIgnition,HazeScale,Param3,LightScale]}
  ShadingPathSwitch(
   Default=LinearInterpolate(A=@n27, B=Multiply(A=@n28, B=@n29.LightRadius), Alpha=LinearInterpolate(A=@n30, B=@n31, Alpha=@n32.isIgnition) {constA=1})
   Forward=@n27
   Mobile=@n27
  )
 ]
}
