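# Additive unlit Niagara particle material that fakes a noisy point light on nearby scene surfaces, tinted by particle color.
# Scene position behind translucency gives a spherical falloff around the particle, multiplied by scene-normal facing, distance fade and particle alpha.
# Screen-stabilized panning noise modulates color, and a normal-weighted fire-noise alpha mask is also built.
# A distance-synced camera offset pulls and shrinks the particle, alongside apparently unused branches for distance-field alignment, UV distortion and sweep masks.
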Material /Game/Curie/Effects/M_Curie_FakeLight_New
graph {
 properties {
  bEnableResponsiveAA=true
  bUsedWithNiagaraMeshParticles=true
  bUsedWithNiagaraSprites=true
  blendMode=BLEND_Additive
  pixelDepthOffsetMode=PDOM_Legacy
  shadingModel=MSM_Unlit
  usedShadingModels=Unlit
 }
 nodes [
  n1 = ParticlePositionWS()
  n2 = WorldPosition() {worldPositionShaderOffset=WPT_ExcludeAllShaderOffsets}
  n3 = ParticleRadius()
  n4 = MaterialFunctionCall(
   Position=ParticlePositionWS()
   Fade_Start_Distance=Add(A=ScalarParameter() {defaultValue=500, parameterName=FadeStartDistance}, B=@n3)
   Fade_End_Distance=Add(A=ScalarParameter() {defaultValue=150, parameterName=FadeEndDistance}, B=@n3)
  ) {materialFunction=&MF_Distance_SyncedFade_FX}
  n5 = LinearInterpolate(Alpha=@n4.Result) {constA=0.5}
  n6 = Multiply(A=@n5, B=ParticleRadius())
  Add(
   A=Multiply(A=Subtract(A=@n1, B=@n2.XYZ), B=OneMinus(None=@n5)) {constB=0}
   B=MaterialFunctionCall(Offset Amount=NamedRerouteDeclaration(None=@n6) {name=Particle Radius, nodeColor=<rgba 1 0.5156 0 1>}) {materialFunction=/Engine/Functions/Engine_MaterialFunctions02/WorldPositionOffset/CameraOffset}.Camera Offset
  ) {constB=0}
  LinearInterpolate(
   A=ScalarParameter() {defaultValue=1, parameterName=UV Distortion Strength, sliderMax=1}
   Alpha=MaterialFunctionCall(Float2=TextureCoordinate()) {materialFunction=&BreakOutFloat2Components}.G
  ) {constA=2}
  n7 = MaterialFunctionCall!(Position=ParticlePositionWS(), UVRadialScale=Divide(B=ParticleRadius()) {constA=100}) {materialFunction=/Game/Weapons/WeaponSkins/MaterialFunction/ObjectStabilizedScreenSpaceUVs}
  n8 = VectorParameter() {parameterName=Smoke Tile Speed, defaultValue=<rgba 1 0.8 0 0.33>}
  n9 = MaterialFunctionCall(X=@n8.B, Y=@n8.B) {materialFunction=&MakeFloat2}
  n10 = TextureSampleParameter2D(UVs=Multiply(A=Panner(Coordinate=@n7.StabilizedUVs, Speed=@n9.Result) {speedX=0.01, speedY=0.66}, B=ComponentMask(None=@n8.RGB) {g=true, r=true})) {
   constMipValue=0
   group=Base Sprite Texture
   parameterName=Smoke Texture
   samplerType=SAMPLERTYPE_LinearColor
   sortPriority=60
   texture=/Game/Global/Textures/Default/Noise/T_Curie_NoisePack
  }
  n11 = ScalarParameter() {defaultValue=0.75, parameterName=Flame Multiplier, sliderMax=5}
  n12 = ParticleColor()
  n13 = Multiply(
   A=Power(Base=LinearInterpolate(B=@n10.B, Alpha=@n11) {constA=1}, Exp=ScalarParameter() {defaultValue=1, parameterName=Color Power, sliderMax=5})
   B=@n12.RGB
  )
  n14 = ScalarParameter() {defaultValue=1, parameterName=Light Intensity}
  n15 = MaterialFunctionCall() {materialFunction=&WorldPositionBehindTranslucency}
  n16 = Length(None=Subtract(A=ParticlePositionWS(), B=@n15.Result)) {is=class MaterialExpressionLength}
  n17 = NamedRerouteUsage()
  n18 = SmoothStep(Min=Min(A=ScalarParameter() {defaultValue=0.25, parameterName=Inner Radius}) {constB=0.99}, Value=Saturate(None=Divide(A=@n16, B=@n17)))
  n19 = Power(
   Base=NamedRerouteDeclaration(None=OneMinus(None=@n18)) {name=Sphere Mask, nodeColor=<rgba 0.4766 0 1 1>}
   Exp=ScalarParameter() {defaultValue=2, parameterName=Light Exponent}
  ) {constExponent=3}
  n20 = Subtract(
   A=ParticlePositionWS()
   B=MaterialFunctionCall() {materialFunction=&WorldPositionBehindTranslucency}.Result
  )
  n21 = ShadingPathSwitch(Default=SceneTexture() {sceneTextureId=PPI_WorldNormal}.Color, Mobile=Constant3Vector()) {desc=Migrated from FeatureLevelSwitch}
  n22 = Saturate(None=DotProduct(A=Normalize(VectorInput=@n20), B=ComponentMask(None=@n21) {b=true, g=true, r=true}))
  n23 = Multiply(A=Multiply(A=NamedRerouteUsage(), B=@n12.A), B=Multiply(A=@n4.Result, B=Multiply(A=@n19, B=@n22)))
  MaterialFunctionCall(Emissive=Multiply(A=Multiply(A=@n13, B=@n14), B=Saturate(None=@n23))) {materialFunction=&MF_FX_EmissiveAdjustments}
  Multiply(A=MaterialFunctionCall!(Center Point World Position) {materialFunction=/Engine/Functions/Engine_MaterialFunctions02/UVs/WorldSpaceAlignedScreenCoordinates}.Offset and Scaled UVS) {constB=5}
  n24 = VectorParameter() {parameterName=Flame Tile Speed, defaultValue=<rgba 1.25 1.25 0 0>}
  n25 = Panner(
   Coordinate=Add(A=TextureCoordinate()) {constB=0.5}
   Speed=MaterialFunctionCall(X=@n24.B, Y=@n24.B) {materialFunction=&MakeFloat2}.Result
  ) {speedX=-0.05, speedY=1}
  n26 = NamedRerouteUsage()
  Multiply(A=ComponentMask(None=@n24.RGB) {g=true, r=true}, B=Add(A=Add(A=@n25, B=@n26), B=NamedRerouteUsage()))
  Multiply(A=ParticleRandom()) {constB=1.6}
  Multiply()
  n27 = ParticlePositionWS()
  n28 = MaterialFunctionCall() {materialFunction=&WorldPositionBehindTranslucency}
  n29 = DotProduct(
   A=Normalize(VectorInput=Subtract(A=@n27, B=@n28.Result))
   B=ComponentMask(None=SceneTexture() {sceneTextureId=PPI_WorldNormal}.Color) {b=true, g=true, r=true}
  )
  n30 = DynamicParameter() {desc="##1", parameterIndex=1, defaultValue=<rgba 0 0 0 1>, paramNames [LightRadius, FlickerSpeed, FlickerScale, LightAlphaScale]}
  n31 = Multiply(A=@n30.LightRadius, B=DynamicParameter() {desc="##3", parameterIndex=3, paramNames [isIgnition,HazeScale,Param3,LightScale]}.LightScale)
  n32 = Reroute()
  n33 = ComponentMask(None=@n32) {b=true}
  n34 = SphereMask(
   A=ParticlePositionWS()
   B=MaterialFunctionCall() {materialFunction=&WorldPositionBehindTranslucency}.Result
   Radius=Multiply(A=Multiply(A=@n31, B=@n33), B=ComponentMask(None=@n32) {r=true})
   Hardness=Constant() {r=0.1}
  )
  n35 = ScalarParameter() {defaultValue=4, parameterName=RadiusFalloffExponent}
  n36 = SceneTexture() {sceneTextureId=PPI_StoredBaseColor}
  Multiply(
   A=Multiply(A=Saturate(None=@n29), B=Power(Base=@n34, Exp=@n35))
   B=LinearInterpolate(B=ComponentMask(None=@n36.Color) {b=true, g=true, r=true}) {constA=1, constAlpha=0.4}
  )
  n37 = ParticlePositionWS()
  n38 = Normalize(VectorInput=DistanceFieldGradient!(World Position=@n37))
  n39 = StaticBoolParameter() {defaultValue=true, parameterName=Use DistanceField}
  NamedRerouteDeclaration(None=StaticSwitch(True=@n38, False=VertexNormalWS(), Value=@n39)) {name=Distance Field Normal, nodeColor=<rgba 0 0.6953 1 1>}
  n40 = Subtract(A=WorldPosition().XYZ, B=@n37)
  n41 = Subtract(
   A=MaterialFunctionCall!(VectorToRotate=@n40, LookAtVector=@n38, RestingVector=Constant3Vector() {constant=<rgba 0 0 1 0>}) {materialFunction=/Engine/Functions/Engine_MaterialFunctions01/Vectors/RotateVector}.Result
   B=@n40
  )
  n42 = Constant3Vector()
  NamedRerouteDeclaration(None=ShadingPathSwitch(Default=StaticSwitch(True=@n41, False=@n42, Value=@n39), Mobile=@n42) {desc=disable for ES3.1}) {name=WPO DF Aligned Edge, nodeColor=<rgba 0.8516 1 0 1>}
  n43 = TextureCoordinate()
  n44 = Constant() {r=0.5}
  n45 = Power(Base=SphereMask(A=@n43, B=@n44, Radius=@n44) {hardnessPercent=0}) {constExponent=1}
  n46 = ScalarParameter() {defaultValue=0.5, parameterName=UV Pinch Amount, sliderMax=1}
  NamedRerouteDeclaration(None=Multiply(A=ConstantBiasScale(None=@n43) {bias=-0.5}, B=Multiply(A=@n45, B=@n46)) {constA=1}) {name=UV Pinch, nodeColor=<rgba 0.4297 0 1 1>}
  n47 = Add(B=TextureCoordinate())
  n48 = VectorParameter() {parameterName=UV Distortion Tile Speed, defaultValue=<rgba 1 1 -0.15 1.5>}
  n49 = MaterialFunctionCall(X=@n48.B, Y=@n48.B) {materialFunction=&MakeFloat2}
  n50 = TextureSampleParameter2D(UVs=Multiply(A=Panner(Coordinate=@n47, Speed=@n49.Result) {speedX=0.01, speedY=1.2}, B=ComponentMask(None=@n48.RGB) {g=true, r=true})) {
   constMipValue=0
   group=Base Sprite Texture
   parameterName=UV Distortion
   samplerType=SAMPLERTYPE_Normal
   sortPriority=60
   texture=/Game/Global/Textures/Default/Noise/T_SwirlNoise_N
  }
  n51 = ScalarParameter() {defaultValue=1, parameterName=UV Distortion Strength, sliderMax=1}
  n52 = Constant2Vector() {g=1, r=1}
  NamedRerouteDeclaration(None=Multiply(A=ComponentMask(None=@n50.RGB) {g=true, r=true}, B=Multiply(A=@n51, B=@n52)) {constB=0.066}) {name=UV Distortion, nodeColor=<rgba 0.1484 0 1 1>}
  n53 = Divide(A=NamedRerouteUsage()) {constB=400}
  n54 = TextureSampleParameter2D(UVs=Add(A=ComponentMask(None=@n53) {g=true, r=true}, B=ParticleRandom())) {
   constMipValue=0
   group=Base Sprite Texture
   parameterName=Fire Texture
   samplerType=SAMPLERTYPE_LinearColor
   sortPriority=60
   texture=&T_Curie_NoisePack
  }
  n55 = Multiply(A=@n54.B, B=ScalarParameter() {defaultValue=2, parameterName=Alpha Multiplier, sliderMax=5})
  n56 = ShadingPathSwitch(Default=SceneTexture() {sceneTextureId=PPI_WorldNormal}.Color, Mobile=Constant3Vector()) {desc=Migrated from FeatureLevelSwitch}
  n57 = Abs(None=DotProduct(A=ComponentMask(None=@n56) {b=true, g=true, r=true}, B=Constant3Vector() {constant=<rgba 0 0 1 0>}))
  NamedRerouteDeclaration(None=LinearInterpolate(B=Power(Base=@n55) {constExponent=1.25}, Alpha=Multiply(A=@n57) {constB=0.5}) {constA=0.66}) {name=Alpha Mask, nodeColor=<rgba 0 0.9766 1 1>}
  NamedRerouteDeclaration(None=@n15.Result) {name=World Space UV, nodeColor=<rgba 1 0.2344 0 1>}
  NamedRerouteUsage()
  ObjectPositionWS()
  ObjectPositionWS()
  n58 = DynamicParameter() {desc="##2", parameterIndex=2, defaultValue=<rgba 0 0 0 175>, paramNames [X, Y, Z, SphericalRadius]}
  n59 = SphereMask(
   A=MaterialFunctionCall(R=@n58.X, G=@n58.X, B=@n58.X) {materialFunction=&AppendMany}.RGB
   B=ParticlePositionWS()
   Radius=@n58.SphericalRadius
   Hardness=Constant()
  )
  n60 = ParticleRelativeTime()
  n61 = SphereMask(
   A=@n59
   B=ConstantBiasScale(None=OneMinus(None=@n60)) {bias=-0.5, scale=2}
   Radius=ScalarParameter() {defaultValue=0.3, parameterName=SweepRadius}
   Hardness=Constant() {r=0.2}
  )
  OneMinus(None=Multiply(A=Ceil(None=@n59), B=OneMinus(None=@n61)))
  OneMinus(None=VectorParameter() {parameterName=WPO Scale, defaultValue=<rgba 1 1 1 0>}.RGB)
  ParticleDirection()
  ParticlePositionWS()
  ParticleRandom()
  ParticleSize()
  ScalarParameter() {defaultValue=100, parameterName=Light Inner Radius}
  SceneTexture() {sceneTextureId=PPI_WorldNormal}
  VectorParameter() {parameterName=Smoke Tile Speed, defaultValue=<rgba 1 0.8 0 0.33>}
 ]
}
