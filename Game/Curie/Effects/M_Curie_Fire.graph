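# Unlit translucent fire material for Niagara sprite and mesh particles in the Curie effects, using a SubUV flipbook tinted by particle color.
# Panning smoke noise, controlled by dynamic parameters, erodes flipbook alpha and varies depth-fade distance.
# Mobile skips depth fade, and a static switch adds near-camera fade.
# An optional ignition branch uses sphere masks and a lifetime sweep to compute an offset pulling vertices toward particle center.
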Material /Game/Curie/Effects/M_Curie_Fire
graph {
 properties {
  bUsedWithNiagaraMeshParticles=true
  bUsedWithNiagaraSprites=true
  blendMode=BLEND_Translucent
  pixelDepthOffsetMode=PDOM_Legacy
  shadingModel=MSM_Unlit
  usedShadingModels=Unlit
 }
 nodes [
  n1 = TextureSampleParameterSubUV!(UVs=TextureCoordinate(), Apply View MipBias) {parameterName=FireSubUV, texture=/Game/Effects/Fort_Effects/Textures/Fire/T_FireFluid_1_8X8}
  MaterialFunctionCall(Emissive=Multiply(A=@n1.R, B=ParticleColor().RGB)) {materialFunction=&MF_FX_EmissiveAdjustments}
  n2 = ScalarParameter() {defaultValue=8, parameterName=FramesX}
  n3 = ScalarParameter() {defaultValue=8, parameterName=FramesY}
  n4 = DynamicParameter() {defaultValue=<rgba 0 1 1 1>, paramNames [LerpNoise,X,PowerErode,Scale]}
  n5 = Add(A=Multiply(A=AppendVector(A=@n2, B=@n3), B=TextureCoordinate()), B=AppendVector(A=@n4.X, B=Add(A=@n4.X) {constB=2}))
  n6 = VectorParameter() {parameterName=Noise Panner Speed, defaultValue=<rgba 0 -0.125 0 0>}
  n7 = TextureSample(UVs=Panner(Coordinate=Multiply(A=@n5, B=@n4.Scale) {desc=Scale}, Speed=ComponentMask(None=@n6.RGB) {g=true, r=true}) {bFractionalPart=true, speedY=-0.125}) {texture=/Game/Effects/Fort_Effects/Textures/Tile/T_Inky_Smoke_Tile2}
  n8 = Power(Base=Multiply(A=@n1.A, B=@n7.G), Exp=DynamicParameter() {defaultValue=<rgba 0 1 1 1>, paramNames [LerpNoise,X,PowerErode,Scale]}.PowerErode) {constExponent=4}
  n9 = DynamicParameter() {defaultValue=<rgba 0 1 1 1>, paramNames [LerpNoise,X,PowerErode,Scale]}
  n10 = Multiply(A=@n1.A, B=Multiply(A=ParticleColor().A, B=LinearInterpolate(A=@n1.A, B=@n8, Alpha=@n9.LerpNoise) {constAlpha=1}))
  n11 = ScalarParameter() {defaultValue=60, parameterName=DepthFade_Low}
  n12 = ScalarParameter() {defaultValue=120, parameterName=DepthFade_High}
  n13 = ShadingPathSwitch(
   Default=DepthFade(Opacity=@n10, FadeDistance=LinearInterpolate(A=@n11, B=@n12, Alpha=@n7.G) {constA=60, constB=120, desc=Bias Noise})
   Mobile=@n10
  ) {desc="Added pref for mobile,Yang."}
  n14 = ScalarParameter() {defaultValue=150, parameterName=Near Camera Distance}
  StaticSwitchParameter(
   True=Multiply(A=@n13, B=MaterialFunctionCall(Distance=@n14) {materialFunction=&Near_Camera_Fade}.Result)
   False=@n13
  ) {parameterName=Use Near Camera Fade}
  n15 = DynamicParameter() {parameterIndex=2, defaultValue=<rgba 0 0 0 175>, paramNames [X, Y, Z, SphericalRadius]}
  n16 = SphereMask(
   A=MaterialFunctionCall(R=@n15.X, G=@n15.X, B=@n15.X) {materialFunction=&AppendMany}.RGB
   B=ParticlePositionWS()
   Radius=@n15.SphericalRadius
   Hardness=Constant()
  )
  n17 = Ceil(None=@n16)
  n18 = OneMinus(None=ParticleRelativeTime())
  n19 = OneMinus(None=SphereMask(
    A=@n16
    B=ConstantBiasScale(None=@n18) {bias=-0.5, scale=2}
    Radius=ScalarParameter() {defaultValue=0.3, parameterName=SweepRadius}
    Hardness=Constant() {r=0.2}
   ))
  n20 = ParticlePositionWS()
  n21 = WorldPosition()
  StaticSwitchParameter(True=Multiply(A=Multiply(A=@n17, B=@n19), B=Subtract(A=@n20, B=@n21.XYZ)), False=Constant()) {defaultValue=true, parameterName=UseIgnition}
 ]
}
