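# Translucent lit Niagara material for packed-flipbook smoke sprites and meshes in the Curie effects.
# A dynamic parameter picks a SubUV flipbook channel as density, with panning, per-particle-offset clump noise as detail.
# Smoke intensity drives opacity and curve-atlas color, which can be un-multiplied by opacity and is tinted by particle color.
# A world-position-offset branch rotates vertices toward an ignition normal, weighted along V, but builds that normal from X alone.
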Material /Game/Curie/Effects/M_Curie_SubUVSmoke_Packed
graph {
 properties {
  bCastRayTracedShadows=false
  bUsedWithNiagaraMeshParticles=true
  bUsedWithNiagaraSprites=true
  blendMode=BLEND_Translucent
  pixelDepthOffsetMode=PDOM_Legacy
  usedShadingModels=Default Lit
 }
 nodes [
  n1 = TextureObjectParameter() {
   group=Flame Channels
   parameterName=Texture
   samplerType=SAMPLERTYPE_LinearColor
   sortPriority=1
   texture=/Game/Effects/Fort_Effects/Textures/Fire/T_2DSmoke_Flipbk_A_packed
  }
  n2 = MaterialFunctionCall(Float2=TextureCoordinate()) {materialFunction=&BreakOutFloat2Components}
  n3 = Power(Base=@n2.G, Exp=ScalarParameter() {defaultValue=1, parameterName=Vertical Exponent, sliderMax=3}) {constExponent=1}
  n4 = DynamicParameter() {paramNames [Density Scale, Packed Channel, SubUV_Images, SubUV_Frame]}
  n5 = MaterialFunctionCall(Float3=MaterialFunctionCall!(
    Texture=@n1
    UVs=MaterialFunctionCall(X=@n2.R, Y=@n3) {materialFunction=&MakeFloat2}.Result
    SubImages=AppendVector(A=@n4.SubUV_Images, B=@n4.SubUV_Images)
    Frame=@n4.SubUV_Frame
   ) {materialFunction=/Engine/Functions/Engine_MaterialFunctions03/Texturing/SubUV_Function}.RGB) {materialFunction=&BreakOutFloat3Components}
  n6 = DynamicParameter() {desc=G - packed channel switch/weight, paramNames [Density Scale, Packed Channel, SubUV_Images, SubUV_Frame]}
  n7 = Frac(None=Clamp(None=@n6.Packed Channel))
  Add(A=@n5.G, B=Multiply(A=@n5.R, B=@n7))
  n8 = DynamicParameter() {parameterIndex=2, paramNames [Ignition Normal.X, Ignition Normal.Y, Ignition Normal.Z, Size]}
  n9 = Multiply(A=TextureCoordinate(), B=@n8.Size)
  n10 = DynamicParameter() {parameterIndex=1, paramNames [Sprite_RandomOffset, Sprite_TimeOffset, Detail_Intensity, Detail_Pan_Speed]}
  n11 = MaterialFunctionCall(X=@n10.Sprite_RandomOffset, Y=@n10.Sprite_TimeOffset) {materialFunction=&MakeFloat2}
  n12 = Constant()
  n13 = ScalarParameter() {defaultValue=0.2, parameterName=Detail PanSpeed, sliderMax=1}
  n14 = Panner(
   Coordinate=Multiply(A=Add(A=@n9, B=@n11.Result), B=ScalarParameter() {defaultValue=1.5, parameterName=Detail TexScale, sliderMax=1}) {constB=0.4}
   Time=Add(A=Time(), B=Multiply(A=@n10.Sprite_RandomOffset) {constB=25})
   Speed=AppendVector(A=@n12, B=Multiply(A=@n8.Size, B=@n13))
  ) {speedY=0.05}
  n15 = Multiply(
   A=DynamicParameter() {parameterIndex=1, paramNames [Sprite_RandomOffset, Sprite_TimeOffset, Detail_Intensity, Detail_Pan_Speed]}.Detail_Intensity
   B=Power(
    Base=TextureSample(UVs=@n14) {texture=/Game/Effects/Fort_Effects/Textures/Smoke/T_Clumps}.R
    Exp=ScalarParameter() {defaultValue=1.3, parameterName=detail exp, sliderMax=2}
   ) {constExponent=1}
  ) {constA=0.5}
  ComponentMask(None=@n15) {r=true}
  n16 = ParticlePositionWS()
  DistanceFieldGradient!(World Position=@n16)
  MaterialFunctionCall!(
   Height Map=TextureObject() {texture=&T_Clumps}
   Normal Map Intensity=ScalarParameter() {defaultValue=1, parameterName=Normal Intensity, sliderMax=1}
   Height Map UV Offset
   Coordinates=NamedRerouteUsage()
   Height Map Channel Selector
  ) {materialFunction=/Engine/Functions/Engine_MaterialFunctions03/Procedurals/NormalFromHeightmap}
  n17 = ParticleColor()
  n18 = Multiply(
   A=NamedRerouteUsage()
   B=LinearInterpolate(
    A=ScalarParameter() {defaultValue=0.02, parameterName=Detail Min, sliderMax=3, sliderMin=-1}
    B=ScalarParameter() {defaultValue=1, parameterName=Detail Max, sliderMax=3, sliderMin=-1}
    Alpha=NamedRerouteUsage()
   ) {constA=0.8, constB=2}
  )
  n19 = Power(Base=@n18, Exp=ScalarParameter() {defaultValue=1, group=Output Color, parameterName=Opacity Exponent, sliderMax=3, sortPriority=8})
  n20 = ScalarParameter() {defaultValue=1.3, group=Output Color, parameterName=Opacity Scale, sliderMax=5, sortPriority=8}
  n21 = Saturate(None=Multiply(A=@n17.A, B=Multiply(A=@n19, B=@n20) {constB=3}))
  n22 = CurveAtlasRowParameter(CurveTime=Power(Base=@n18, Exp=ScalarParameter() {defaultValue=1.2, parameterName=Color Exp, sliderMax=4}) {constExponent=1}) {
   atlas=/Game/Curie/Effects/temp/Curves/CA_Curie_Fire_Visuals
   curve=/Game/Curie/Effects/temp/Curves/Curve_Smoky
   defaultValue=7
   parameterName=Curves
  }
  n23 = Divide(A=@n22, B=@n21)
  n24 = ScalarParameter() {defaultValue=1, group=Output Color, parameterName=UnMult, sliderMax=1}
  Multiply(
   A=@n17.RGB
   B=Multiply(
    A=LinearInterpolate(A=@n22, B=@n23, Alpha=@n24)
    B=ScalarParameter() {defaultValue=0.15, group=Output Color, parameterName=Final Color Brightness, sliderMax=50, sortPriority=9}
   )
  )
  NamedRerouteDeclaration(None=Subtract(A=@n14, B=AppendVector(A=@n12, B=@n10.Detail_Pan_Speed))) {name=UVCoords, nodeColor=<rgba 1 0 0.2578 1>}
  n25 = Multiply(
   A=Multiply(A=@n6.Density Scale, B=If(A=@n6.Packed Channel, "A > B"=@n5.B, "A == B"=@n5.G, "A < B"=@n5.R) {constB=1, equalsThreshold=0.1})
   B=ScalarParameter() {defaultValue=1, group=Smoke Channels, parameterName=Density Channel Scale, sliderMax=10, sortPriority=3}
  )
  n26 = ScalarParameter() {defaultValue=1, group=Smoke Channels, parameterName=Density Exponent, sliderMax=3, sortPriority=5}
  NamedRerouteDeclaration(None=Multiply(
    A=Power(Base=@n25, Exp=@n26)
    B=ScalarParameter() {defaultValue=1, group=Smoke Channels, parameterName=Density Scale, sliderMax=25, sortPriority=5}
   )) {name=Flipbook Smoke Texture, nodeColor=<rgba 1 0.5391 0 1>}
  n27 = DynamicParameter() {parameterIndex=2, paramNames [Ignition Normal.X, Ignition Normal.Y, Ignition Normal.Z, Size]}
  n28 = MaterialFunctionCall(X=@n27.Ignition Normal.X, Y=@n27.Ignition Normal.X, Z=@n27.Ignition Normal.X) {materialFunction=&MakeFloat3}
  n29 = Subtract(A=WorldPosition() {worldPositionShaderOffset=WPT_ExcludeAllShaderOffsets}.XYZ, B=@n16)
  n30 = Subtract(
   A=MaterialFunctionCall!(VectorToRotate=@n29, LookAtVector=Normalize(VectorInput=@n28.Result), RestingVector=Constant3Vector() {constant=<rgba 0 0 1 0>}) {materialFunction=/Engine/Functions/Engine_MaterialFunctions01/Vectors/RotateVector}.Result
   B=@n29
  )
  n31 = ScalarParameter() {defaultValue=0.3, parameterName=Stick Min, sliderMax=1}
  n32 = ScalarParameter() {defaultValue=1, parameterName=Stick Max, sliderMax=1}
  n33 = Power(
   Base=MaterialFunctionCall(Float2=TextureCoordinate()) {materialFunction=&BreakOutFloat2Components}.G
   Exp=ScalarParameter() {defaultValue=0.76, parameterName=Stick Exponent, sliderMax=5}
  ) {constExponent=0.75}
  NamedRerouteDeclaration(None=LinearInterpolate(A=Constant3Vector(), B=Add(A=@n30) {constB=0}, Alpha=LinearInterpolate(A=@n31, B=@n32, Alpha=@n33) {constA=0.3})) {name=WPO DF Aligned Edge, nodeColor=<rgba 1 0.0938 0 1>}
  NamedRerouteDeclaration(None=@n15) {name=Tiling Smoke Texture, nodeColor=<rgba 1 0 0.3516 1>}
  NamedRerouteUsage()
  OneMinus(None=@n7)
  n34 = ScalarParameter() {defaultValue=55, parameterName=FadeDistance, sliderMax=300}
  Reroute(None=OneMinus(None=DepthFade(Opacity=@n21, FadeDistance=@n34)))
  n35 = DistanceFieldGradient(World Position=@n16)
  n36 = DistanceToNearestSurface!(World Position=@n16)
  Reroute(None=Multiply(A=Normalize(VectorInput=@n35), B=Multiply(A=@n36) {constB=-1}))
  TextureSample(UVs=TextureCoordinate()) {texture=/Game/Effects/Fort_Effects/Textures/Debug/RGB_Debug}
 ]
}
