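# Niagara fire effect for Fortnite's Curie fire system, with ignition, flames, flame licks and embers, plus radial area-of-effect fire and ember variants.
# Shared Curie modules handle spawning and bounds-scaled initialization, with a Curie fire data interface exposed as a user variable.
# Curves drive SubUV animation, color and dynamic material parameters across sprite and mesh renderers.
# Both main fire emitters also compute Janus lights for component renderers, with user-set color and size.
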NiagaraSystem /Game/Curie/Effects/NS_Curie_Fire
dependencies {
 usedDataInterfaces [
  &NiagaraDataInterfaceCurve
  &NiagaraDataInterfaceVectorCurve
  &NiagaraDataInterfaceColorCurve
  &NiagaraDataInterfaceCamera
 ]
 usedDynamicInputs [&VectorFromCurve]
 usedModules [&SystemState]
}
emitters {
 Embers {
  emitterUpdateScript [
   &EmitterState(Life Cycle Mode=::NewEnumerator1, Loop Duration=1, UseLoopDelay=false)
   &SpawnRate(SpawnRate=0, Use Spawn Probability=false) {enabled=false}
   /Game/Effects/Niagara/Modules/Curie/FN_Curie_Fire_Spawn(SpawnCountScale:NiagaraFloat=$Engine.Emitter.SpawnCountScale)
  ]
  particleSpawnScript [
   /Niagara/Modules/Spawn/Initialization/InitializeParticle(
    Color=<rgba 30 13.125 3 1>
    Lifetime=&UniformRangedFloat(Minimum=0.2, Maximum=1.75)
    Module.Write Color=true
    Module.Write Lifetime=true
    Module.Write Mass=true
    Module.Write Material Random=true
    Module.Write Position=false
    Module.Write Scale=false
    Module.Write SpriteRotation=false
    Module.Write SpriteSize=true
    Position=&SimulationPosition()
    Sprite Size=&UniformRangedVector2D(Minimum=<2.5 1>, Maximum=<3.5 2>)
   )
   &AddVelocity(Velocity=&UniformRangedVector(Minimum=<-50 -50 350>, Maximum=<50 50 1>))
   /Game/Effects/Niagara/Modules/Curie/FN_Curie_Fire_InitializePosition()
   &SphereLocation(
    MaskSpawnGroup=false
    Override Seed=false
    Sphere Radius=35
    Surface Only=false
    UseNonUniformScale=false
    UseOffset=false
    UseSphereOrientation=false
   )
  ]
  particleUpdateScript [
   &ParticleState()
   /Niagara/Modules/Update/Forces/CurlNoiseForce(
    Noise Frequency=300
    Noise Strength=&Multiply_Float(
     A=$Particles.MaterialRandom
     B=&FloatFromCurve(
      FloatCurve=DataInterface() {propertyValues {Curve {keys [
          <at 0 4147.9209 RCIM_Cubic>
          {arriveTangent=998.7806, interpMode=RCIM_Cubic, leaveTangent=998.7806, time=0.897, value=2308.9619}
          {arriveTangent=839.8494, interpMode=RCIM_Cubic, leaveTangent=839.8494, time=2.2666, value=6411.7881}
          {arriveTangent=818.7506, interpMode=RCIM_Cubic, leaveTangent=818.7506, time=3.6727, value=4640.1284}
          {arriveTangent=-197.3881, interpMode=RCIM_Cubic, leaveTangent=-197.3881, time=4.7165, value=8417.627}
          {arriveTangent=-812.5138, interpMode=RCIM_Cubic, leaveTangent=-812.5138, time=6.6551, value=4051.4272}
          {arriveTangent=379.0706, interpMode=RCIM_Cubic, leaveTangent=379.0706, time=9.3392, value=4661.6182}
          {arriveTangent=-88.1153, interpMode=RCIM_Cubic, leaveTangent=-88.1153, time=11.8318, value=6013.7651}
          <at 15.169 4147.9209 RCIM_Cubic>
         ]}}}
      CurveIndex=&Add_Float(A=$Engine.System.Age, B=$Particles.Age)
      Scale Curve=0.7
     )
    )
    Pan Noise=true
    Pan Noise Field=<0 0 1000>
   )
   &AccelerationForce(Acceleration=<0 0 250>)
   &Drag(Drag=0.8, Use Linear Drag=true, Use Rotational Drag=false)
   &SolveForcesAndVelocity(Clamp Velocity=false, Limit Acceleration=false)
   &ScaleColor(
    Scale Alpha=&FloatFromCurve(FloatCurve=DataInterface() {propertyValues {Curve {keys [<at 0 0>,<at 0.1997 0.9273>,<at 1.0011 0.002>]}}})
    ScaleA=true
    ScaleRGB=true
   )
   &ScaleSpriteSizeBySpeed(Max Scale Factor=<2 15>, Velocity Threshold=1300)
  ]
  rendererClasses [&NiagaraSpriteRendererProperties]
  renderers [{rendererClass=&NiagaraSpriteRendererProperties}]
 }
 EmbersAOE {
  emitterUpdateScript [
   &EmitterState()
   &SpawnRate(SpawnRate=100, Use Spawn Probability=false) {enabled=false}
   /Game/Effects/Niagara/Modules/Curie/FN_Curie_Fire_Spawn_Radial()
  ]
  particleSpawnScript [
   &InitializeParticle(
    Color=<rgba 2 1.2022 0.6482 1>
    Lifetime=&UniformRangedFloat(Minimum=0.5, Maximum=2)
    Module.Write Color=true
    Module.Write Lifetime=true
    Module.Write Mass=false
    Module.Write Material Random=true
    Module.Write Position=false
    Module.Write Scale=false
    Module.Write SpriteRotation=true
    Module.Write SpriteSize=true
    Position=&SimulationPosition()
    Sprite Rotation=&UniformRangedFloat(Minimum=-180, Maximum=180)
    Sprite Size=&Vector2DFromFloat(Value=&UniformRangedFloat(Minimum=3, Maximum=5))
   )
   &AddVelocity(Velocity=&MakeVector(X=0, Y=0, Z=&UniformRangedFloat(Minimum=50, Maximum=150)))
   &ApplyInitialForces()
   /Niagara/Modules/Spawn/Location/SphereLocation(MaskSpawnGroup=false, Module.Override Randomness=false, Sphere Radius=50, Surface Only=false) {enabled=false}
   &AddVelocityFromPoint(Module.Use Falloff Distance=false, Velocity Strength=&UniformRangedFloat(Minimum=60, Maximum=100))
   /Game/Effects/Niagara/Modules/Curie/FN_Curie_Fire_InitializePosition_Radial()
  ]
  particleUpdateScript [
   &ParticleState()
   &AccelerationForce(Acceleration=<0 0 500>)
   &CurlNoiseForce(
    Noise Frequency=0.016
    Noise Strength=&FloatFromCurve(FloatCurve=DataInterface() {propertyValues {Curve {keys [<at 0 1>,<at 1 0>]}}}, Scale Curve=1200)
    Pan Noise=true
    Pan Noise Field=<0 0 1>
   )
   &Drag(Drag=3, Use Linear Drag=true, Use Rotational Drag=false)
   &SolveForcesAndVelocity(Clamp Velocity=false, Limit Acceleration=false)
   &ScaleColor(
    Scale Alpha=&FloatFromCurve(FloatCurve=DataInterface() {propertyValues {Curve {keys [
         {arriveTangent=11.9433, interpMode=RCIM_Cubic, leaveTangent=11.9433, tangentMode=RCTM_User, time=0, value=0}
         <at 0.2 1 RCIM_Cubic>
         <at 1 0 RCIM_Cubic>
        ]}}})
    Scale RGB=&MakeVectorFromLinearColorRGB(Color=&ColorFromCurve(ColorCurve=DataInterface() {propertyValues {
        AlphaCurve {keys [<at 0 1 RCIM_Cubic>]}
        BlueCurve {keys [<at 0 0.5069 RCIM_Cubic>,<at 1 0 RCIM_Cubic>]}
        GreenCurve {keys [<at 0 1.6422 RCIM_Cubic>,<at 1 0.06 RCIM_Cubic>]}
        RedCurve {keys [<at 0 4 RCIM_Cubic>,<at 1 1 RCIM_Cubic>]}
       }}))
    ScaleA=true
    ScaleRGB=true
   )
   &DynamicMaterialParameters(
    Index 0 Param 1=&FloatFromCurve(FloatCurve=DataInterface() {propertyValues {Curve {keys [<at 0 0>,<at 1 1>]}}})
    Index 0 Param 2=1
    Index 0 Param 3=1
    Index 0 Param 4=1
    Module.Param0WriteEnabled=true
    Module.Param1WriteEnabled=true
    Module.Param2WriteEnabled=true
    Module.Param3WriteEnabled=true
   )
   &ScaleSpriteSize(Scale Factor=&Vector2DFromFloat(Value=&FloatFromCurve(FloatCurve=DataInterface() {propertyValues {Curve {keys [
          {arriveTangent=9.1531, interpMode=RCIM_Cubic, leaveTangent=9.1531, tangentMode=RCTM_User, time=0, value=0.02}
          {arriveTangent=0.48, interpMode=RCIM_Cubic, leaveTangent=0.48, time=0.18, value=0.99}
          <at 1 0.5 RCIM_Cubic>
         ]}}})))
   &ScaleSpriteSizeBySpeed(Max Scale Factor=<1 4>)
  ]
  rendererClasses [&NiagaraSpriteRendererProperties]
  renderers [{rendererClass=&NiagaraSpriteRendererProperties}]
 }
 Fire {
  emitterUpdateScript [
   &EmitterState(Life Cycle Mode=::NewEnumerator1, Loop Duration=1, UseLoopDelay=false)
   &FN_Curie_Fire_Spawn()
  ]
  particleSpawnScript [
   &InitializeParticle(
    Color=<rgba 1 1 1 0.4>
    Lifetime=&UniformRangedFloat(Minimum=0.6, Maximum=0.9)
    Module.Write Color=true
    Module.Write Lifetime=true
    Module.Write Mass=false
    Module.Write Material Random=true
    Module.Write Position=false
    Module.Write Scale=false
    Module.Write SpriteRotation=false
    Module.Write SpriteSize=true
    Position=&SimulationPosition()
    Sprite Rotation=&UniformRangedFloat(Minimum=0, Maximum=360)
    Sprite Size=&Vector2DFromFloat(Value=100)
   )
   /SetVariables(Particles.RandomFloat=&UniformAOrBFloat(A=-1)) {isSetParameters=true}
   /Game/Effects/Niagara/Modules/Curie/FN_Curie_Fire_InitializeSize(ParticleSizeMax=450, ParticleSizeMin=200, SphericalRadiusRangeInputHigh=900)
   &FN_Curie_Fire_InitializePosition()
   /Game/Effects/Niagara/Modules/Curie/FN_Curie_Fire_InitializeAcceleration()
   /Game/Effects/Niagara/Modules/Curie/FN_Curie_Fire_InitializeBoundsRadius()
   /Game/Effects/Niagara/Modules/Curie/FN_Curie_Fire_InitializeSubUV(RandomSubUVStartingFrameMax:NiagaraFloat=$User.NumberOfSubUVFrames)
   /SetVariables(Particles.Position=&Subtract_Vector(A=Linked() {linkedVariable {name=Particles.Position, type=NiagaraPosition}}, B=<0 0 50>)) {isSetParameters=true}
   &AddVelocity(Coordinate Space=::World, Velocity=<0 0 5>)
   &ApplyInitialForces()
   &DynamicMaterialParameters(
    Index 0 Param 1=3
    Index 0 Param 2=0.5
    Index 0 Param 3=&UniformRangedFloat(Minimum=1, Maximum=2)
    Index 0 Param 4=&Ceil_Float(Float=&UniformRangedFloat(Minimum=1.3, Maximum=2))
    Index 1 Param 1=&Multiply_Float(A=$Particles.SizeScaledByBoundsFloat, B=2)
    Index 1 Param 2=&UniformRangedFloat(Minimum=0, Maximum=0)
    Index 1 Param 4=0.13
    Index 3 Param 4=&self:SetRadiusBiasByScalability_01()
    Module.Param0WriteEnabled=false
    Module.Param0WriteEnabled1=true
    Module.Param0WriteEnabled2=true
    Module.Param0WriteEnabled3=true
    Module.Param1WriteEnabled=true
    Module.Param1WriteEnabled1=true
    Module.Param1WriteEnabled2=true
    Module.Param1WriteEnabled3=true
    Module.Param2WriteEnabled=true
    Module.Param2WriteEnabled1=true
    Module.Param2WriteEnabled2=true
    Module.Param2WriteEnabled3=true
    Module.Param3WriteEnabled=true
    Module.Param3WriteEnabled1=true
    Module.Param3WriteEnabled2=true
    Module.Param3WriteEnabled3=true
    Write Parameter Index 1=true
    Write Parameter Index 2=true
    Write Parameter Index 3=true
   ) {module=DynamicMaterialParameters002}
   /SetVariables(Particles.SpriteSize=<0 0>) {isSetParameters=true, module=SetVariables_#1}
   /SetVariables(Particles.UVScale:Vector2f=&MakeVector2D(X=&UniformAOrBFloat(A=-1), Y=1)) {isSetParameters=true}
  ]
  particleUpdateScript [
   &ParticleState()
   &AccelerationForce(Acceleration=$Particles.AccelerationScaledByBounds)
   &SolveForcesAndVelocity(Clamp Velocity=false, Limit Acceleration=false)
   &DynamicMaterialParameters(
    Index 0 Param 1=&FloatFromCurve(
     FloatCurve=DataInterface() {propertyValues {Curve {keys [
         {arriveTangent=-7.5742, interpMode=RCIM_Cubic, leaveTangent=-7.5742, tangentMode=RCTM_User, time=0.0915, value=1}
         <at 0.25 1.0092 RCIM_Cubic>
         {arriveTangent=-0.4591, interpMode=RCIM_Cubic, leaveTangent=-0.4591, tangentMode=RCTM_User, time=0.4, value=1}
        ]}}}
     Scale Curve=1.2
    )
    Index 0 Param 2=1
    Index 0 Param 3=&FloatFromCurve(FloatCurve=DataInterface() {propertyValues {Curve {keys [
         {arriveTangent=-0.0921, interpMode=RCIM_Cubic, time=0, value=0.2}
         {arriveTangent=1.8, interpMode=RCIM_Cubic, leaveTangent=1.8, time=0.53, value=0.4}
         <at 1 2>
        ]}}})
    Index 0 Param 4=1
    Index 1 Param 1=&Multiply_Float(
     A=$Particles.SizeScaledByBoundsFloat
     B=&FloatFromCurve(FloatCurve=DataInterface() {propertyValues {Curve {keys [
          {arriveTangent=3.3396, interpMode=RCIM_Cubic, leaveTangent=3.3396, tangentMode=RCTM_User, time=0, value=0.5}
          {arriveTangent=1.5, interpMode=RCIM_Cubic, leaveTangent=1.5, time=0.2393, value=1.3608}
          <at 1 2>
         ]}}})
    )
    Index 1 Param 4=0.1
    Index 2 Param 4=&FloatFromCurve(
     FloatCurve=DataInterface() {propertyValues {Curve {keys [
         <at 0.0177 0.0082 RCIM_Cubic>
         {arriveTangent=-0.0064, interpMode=RCIM_Cubic, leaveTangent=-0.0064, time=0.3544, value=0.8348}
         <at 1.0011 0.002>
        ]}}}
     Scale Curve=&MakeFloatFromLinearColor(LinearColor=$User.LightColor, Channel:ENiagaraLinearColor_Channels=::NewEnumerator3)
    )
    Module.Param0WriteEnabled=true
    Module.Param0WriteEnabled1=true
    Module.Param0WriteEnabled2=true
    Module.Param1WriteEnabled=false
    Module.Param1WriteEnabled1=true
    Module.Param1WriteEnabled2=true
    Module.Param2WriteEnabled=true
    Module.Param2WriteEnabled1=true
    Module.Param2WriteEnabled2=true
    Module.Param3WriteEnabled=false
    Module.Param3WriteEnabled1=true
    Module.Param3WriteEnabled2=false
    Write Parameter Index 1=true
    Write Parameter Index 2=true
   )
   /SetVariables(Particles.SubImageIndex=&FloatFromCurve(
     FloatCurve=DataInterface() {propertyValues {Curve {keys [<at 0 22>,<at 1 48>]}}}
     CurveIndex=&ModuloFloat(
      A=&Multiply_Float(
       A=&Add_Float(A=$Particles.NormalizedAge, B=$Particles.MaterialRandom)
       B=&RandomRangeFloat(Minimum=0.8, Maximum=1.2)
      )
      B=1
     )
    )) {isSetParameters=true, module=SetVariables_#2}
   &CameraOffset(Camera Offset Amount=35)
   /SetVariables(Particles.Scale=/Niagara/DynamicInputs/ScaleByCurve/ScaleVectorByCurve(
     InitialValue:Vector3f=&Multiply_VectorByFloat(Vector=&VectorFromFloat(Value=$Particles.SizeScaledByBoundsFloat), Float=1.5)
     ScaleCurve:NiagaraDataInterfaceVectorCurve=DataInterface() {propertyValues {
       XCurve {keys [{arriveTangent=0.4, interpMode=RCIM_Cubic, time=0, value=0.8}, <at 1 1 RCIM_Cubic>]}
       YCurve {keys [{arriveTangent=0.4, interpMode=RCIM_Cubic, time=0, value=0.8}, <at 1 1 RCIM_Cubic>]}
       ZCurve {keys [{arriveTangent=0.4, interpMode=RCIM_Cubic, time=0, value=0.8}, <at 1 1 RCIM_Cubic>]}
      }}
     CurveIndex:NiagaraFloat=$Particles.NormalizedAge
    )) {isSetParameters=true, module=SetVariables_#3}
   &ScaleColor(
    Scale Mode=::NewEnumerator1
    Scale RGBA=&Vector4FromCurve(Vector4Curve=DataInterface() {propertyValues {
       AlphaCurve {keys [<at 0.0024 0>,<at 0.35 1>,<at 0.67 0.67>,<at 0.9977 -0.0026>]}
       BlueCurve {keys [<at 0.3937 0.2>,<at 1 0>]}
       GreenCurve {keys [<at 0.3937 0.5728>,<at 1 0.575>]}
       RedCurve {keys [<at 0.3937 4>,<at 1 11.5>]}
      }})
    ScaleRGBA=true
   )
   /SetVariables(Particles.Color=&MultiplyLinearColor(
     Linear Color=$Particles.Color
     Scale Factor=&MakeVector4_FromFloat(Float=&RemapRange(
       Input Value To Remap=&VectorLength(Vector=$Particles.Velocity)
       Input Min=50
       Input Max=125
       Output Min=1
       Output Max=0.3
      ))
    )) {isSetParameters=true}
   /SetVariables(Particles.SpriteSize=&Vector2DFromFloat(Value=&Multiply_Float(
      A=$Particles.SizeScaledByBoundsFloat
      B=&FloatFromCurve(FloatCurve=DataInterface() {propertyValues {Curve {keys [
           {arriveTangent=1.9145, interpMode=RCIM_Cubic, leaveTangent=0.3645, tangentMode=RCTM_User, time=0.0096, value=0.3883}
           {arriveTangent=0.4555, interpMode=RCIM_Cubic, leaveTangent=0.4555, time=0.1486, value=0.5288}
           <at 0.9899 0.8348>
          ]}}})
     ))) {isSetParameters=true, module=SetVariables_#4}
   &self:JanusLightCalculations_01(
    JanusLightOffsetToCamera=100
    New Camera Query=DataInterface() {propertyValues {PlayerControllerIndex=0, bRequireCurrentFrameData=false}}
    SpawnLightsFraction=3
   ) {module=JanusLightCalculations_01}
   /SetVariables(Particles.SpriteSizeJanus:Vector2f=&Multiply_Vector2(A=$Particles.SpriteSize, B=&Vector2DFromFloat(Value=$User.JanusSpriteSizeScale))) {isSetParameters=true}
   /SetVariables(Particles.JanusLightIntensity:NiagaraFloat=&FloatFromCurve(FloatCurve=DataInterface() {propertyValues {Curve {keys [<at 0 0 RCIM_Cubic>,<at 0.52 1 RCIM_Cubic>,<at 1 0 RCIM_Cubic>]}}}, Scale Curve=2.5)) {isSetParameters=true}
  ]
  rendererClasses [
   &NiagaraMeshRendererProperties
   &NiagaraSpriteRendererProperties
   &NiagaraComponentRendererProperties
  ]
  renderers [
   {rendererClass=&NiagaraMeshRendererProperties, rendererIndex=0}
   {rendererClass=&NiagaraSpriteRendererProperties, rendererIndex=1}
   {rendererClass=&NiagaraComponentRendererProperties, rendererIndex=2}
   {rendererClass=&NiagaraSpriteRendererProperties, rendererIndex=3}
  ]
 }
 FireAOE {
  emitterUpdateScript [
   &EmitterState(Life Cycle Mode=::NewEnumerator1, Loop Duration=1, UseLoopDelay=false)
   &FN_Curie_Fire_Spawn_Radial(SpawnCountScale=0.4)
  ]
  particleSpawnScript [
   &InitializeParticle(
    Color=<rgba 1 1 1 0.35>
    Lifetime=&UniformRangedFloat(Minimum=0.5, Maximum=1)
    Module.Write Color=true
    Module.Write Lifetime=true
    Module.Write Mass=false
    Module.Write Material Random=true
    Module.Write Position=false
    Module.Write Scale=false
    Module.Write SpriteRotation=true
    Module.Write SpriteSize=true
    Position=&SimulationPosition()
    Sprite Rotation=&UniformRangedFloat(Minimum=-180, Maximum=180)
    Sprite Size=&Vector2DFromFloat(Value=&UniformRangedFloat(
      Minimum=&MakeCustomFloatFromBool(Bool=&SetBoolByFloatComparison(A=$Particles.MaterialRandom, B=0.8), True Float=150, False Float=75)
      Maximum=&MakeCustomFloatFromBool(Bool=&SetBoolByFloatComparison(A=$Particles.MaterialRandom, B=0.8), True Float=200, False Float=150)
     ))
   ) {module=InitializeParticle001}
   /SetVariables(Particles.SizeScaledByBoundsFloat:NiagaraFloat=&UniformRangedFloat(Minimum=145, Maximum=170)) {enabled=false, isSetParameters=true}
   /SetVariables(Particles.SizeScaledByBoundsFloat:NiagaraFloat=&RemapRange(Input Value To Remap=$Emitter.Age, Input Min=5, Input Max=12, Output Min=180, Output Max=100)) {isSetParameters=true}
   &FN_Curie_Fire_InitializePosition_Radial()
   &FN_Curie_Fire_InitializeBoundsRadius()
   &FN_Curie_Fire_InitializeSubUV(RandomSubUVStartingFrameMax:NiagaraFloat=$User.NumberOfSubUVFrames)
   &ApplyInitialForces()
   &AddVelocity(Velocity=&MakeVector(
     X=0
     Y=0
     Z=&UniformRangedFloat(
      Minimum=25
      Maximum=&MakeCustomFloatFromBool(Bool=&SetBoolByFloatComparison(A=$Particles.MaterialRandom, B=0.8), True Float=200, False Float=50)
     )
    )) {module=AddVelocity001}
   &AddVelocityFromPoint(
    Module.Use Falloff Distance=false
    Velocity Strength=&MakeCustomFloatFromBool(Bool=&SetBoolByFloatComparison(A=$Particles.MaterialRandom, B=0.8), True Float=100, False Float=0)
   )
   &DynamicMaterialParameters(
    Index 0 Param 1=3
    Index 0 Param 2=0.5
    Index 0 Param 3=&UniformRangedFloat(Minimum=1, Maximum=2)
    Index 0 Param 4=&Ceil_Float(Float=&UniformRangedFloat(Minimum=1.3, Maximum=2))
    Index 1 Param 1=&Multiply_Float(A=$Particles.SizeScaledByBoundsFloat, B=2)
    Index 1 Param 2=0
    Index 1 Param 4=0.3
    Index 3 Param 4=&self:SetRadiusBiasByScalability()
    Module.Param0WriteEnabled=false
    Module.Param0WriteEnabled1=true
    Module.Param0WriteEnabled2=true
    Module.Param0WriteEnabled3=true
    Module.Param1WriteEnabled=true
    Module.Param1WriteEnabled1=true
    Module.Param1WriteEnabled2=true
    Module.Param1WriteEnabled3=true
    Module.Param2WriteEnabled=true
    Module.Param2WriteEnabled1=true
    Module.Param2WriteEnabled2=true
    Module.Param2WriteEnabled3=true
    Module.Param3WriteEnabled=true
    Module.Param3WriteEnabled1=true
    Module.Param3WriteEnabled2=true
    Module.Param3WriteEnabled3=true
    Write Parameter Index 1=true
    Write Parameter Index 2=true
    Write Parameter Index 3=true
   ) {module=DynamicMaterialParameters001}
   /SetVariables(Particles.SpriteSize=<0 0>) {isSetParameters=true, module=SetVariables_#1}
  ]
  particleUpdateScript [
   &ParticleState()
   &AccelerationForce(Acceleration=<0 0 125>) {module=AccelerationForce001}
   &SolveForcesAndVelocity(Clamp Velocity=false, Limit Acceleration=false)
   &DynamicMaterialParameters(
    Index 0 Param 1=&FloatFromCurve(FloatCurve=DataInterface() {propertyValues {Curve {keys [
         {arriveTangent=-7.5742, interpMode=RCIM_Cubic, leaveTangent=-7.5742, tangentMode=RCTM_User, time=0.0915, value=1}
         <at 0.25 0.2 RCIM_Cubic>
         {arriveTangent=5.7253, interpMode=RCIM_Cubic, leaveTangent=5.7253, tangentMode=RCTM_User, time=0.4, value=1}
        ]}}})
    Index 0 Param 2=1
    Index 0 Param 3=&FloatFromCurve(FloatCurve=DataInterface() {propertyValues {Curve {keys [
         {arriveTangent=-0.0921, interpMode=RCIM_Cubic, time=0, value=0.2}
         {arriveTangent=1.8, interpMode=RCIM_Cubic, leaveTangent=1.8, time=0.53, value=0.4}
         <at 1 2>
        ]}}})
    Index 0 Param 4=1
    Index 1 Param 1=&Multiply_Float(
     A=$Particles.SizeScaledByBoundsFloat
     B=&FloatFromCurve(FloatCurve=DataInterface() {propertyValues {Curve {keys [
          {arriveTangent=3.3396, interpMode=RCIM_Cubic, leaveTangent=3.3396, tangentMode=RCTM_User, time=0, value=0.5}
          {arriveTangent=1.5, interpMode=RCIM_Cubic, leaveTangent=1.5, time=0.2393, value=1.3608}
          <at 1 2>
         ]}}})
    )
    Index 1 Param 4=0.35
    Index 2 Param 4=&FloatFromCurve(
     FloatCurve=DataInterface() {propertyValues {Curve {keys [
         <at 0.0177 0.0082 RCIM_Cubic>
         {arriveTangent=-0.0064, interpMode=RCIM_Cubic, leaveTangent=-0.0064, time=0.3544, value=0.8348}
         <at 1.0011 0.002>
        ]}}}
     Scale Curve=&MakeFloatFromLinearColor(LinearColor=$User.LightColor, Channel:ENiagaraLinearColor_Channels=::NewEnumerator3)
    )
    Module.Param0WriteEnabled=true
    Module.Param0WriteEnabled1=true
    Module.Param0WriteEnabled2=true
    Module.Param1WriteEnabled=false
    Module.Param1WriteEnabled1=true
    Module.Param1WriteEnabled2=true
    Module.Param2WriteEnabled=true
    Module.Param2WriteEnabled1=true
    Module.Param2WriteEnabled2=true
    Module.Param3WriteEnabled=false
    Module.Param3WriteEnabled1=true
    Module.Param3WriteEnabled2=false
    Write Parameter Index 1=true
    Write Parameter Index 2=true
   ) {module=DynamicMaterialParameters003}
   /Niagara/Modules/Update/SubUV/SubUVAnimation(Number Of Frames=&FloatFromCurve(
     FloatCurve=DataInterface() {propertyValues {Curve {keys [<at 0 0>,<at 1.0068 55>]}}}
     Scale Curve=&RandomRangeFloat(Minimum=0.8, Maximum=1.2)
    ))
   /SetVariables(Particles.SubImageIndex=&self:AddRandomFloatToSubimageFrame_2(RandomSubUVStartingFrame:NiagaraFloat=$Particles.RandomSubUVStartingFrame)) {isSetParameters=true, module=SetVariables_#2}
   &CameraOffset(Camera Offset Amount=35)
   /SetVariables(Particles.Scale=&ScaleVectorByCurve(
     InitialValue:Vector3f=&Multiply_VectorByFloat(Vector=&VectorFromFloat(Value=$Particles.SizeScaledByBoundsFloat), Float=1.5)
     ScaleCurve:NiagaraDataInterfaceVectorCurve=DataInterface() {propertyValues {
       XCurve {keys [<at 0 0.4 RCIM_Cubic>, {interpMode=RCIM_Cubic, leaveTangent=0.6, time=0.2, value=0.7}, <at 0.8 1 RCIM_Cubic>, <at 1 0.6>]}
       YCurve {keys [<at 0 0.4 RCIM_Cubic>, {interpMode=RCIM_Cubic, leaveTangent=0.6, time=0.2, value=0.7}, <at 0.8 1 RCIM_Cubic>, <at 1 0.6>]}
       ZCurve {keys [<at 0 0.4 RCIM_Cubic>, {interpMode=RCIM_Cubic, leaveTangent=0.6, time=0.2, value=0.7}, <at 0.8 1 RCIM_Cubic>, <at 1 0.6>]}
      }}
     CurveIndex:NiagaraFloat=$Particles.NormalizedAge
    )) {isSetParameters=true, module=SetVariables_#3}
   &ScaleColor(
    Scale Mode=::NewEnumerator1
    Scale RGBA=&Vector4FromCurve(Vector4Curve=DataInterface() {propertyValues {
       AlphaCurve {keys [<at 0.0024 0>,<at 0.2382 1>,<at 0.67 0.67>,<at 0.9977 -0.0026>]}
       BlueCurve {keys [<at 0.288 0.75>,<at 0.8464 0>,<at 0.9897 0>]}
       GreenCurve {keys [<at 0.288 2.1479>,<at 0.8464 0.5>,<at 0.9897 0.0099>]}
       RedCurve {keys [<at 0.288 15>,<at 0.8464 10>,<at 0.9897 0.0573>]}
      }})
    ScaleRGBA=true
   ) {module=ScaleColor001}
   /SetVariables(Particles.Color=&MultiplyLinearColor(
     Linear Color=$Particles.Color
     Scale Factor=&MakeVector4_FromFloat(Float=&RemapRange(
       Input Value To Remap=&VectorLength(Vector=$Particles.Velocity)
       Input Min=50
       Input Max=125
       Output Min=1
       Output Max=0.3
      ))
    )) {isSetParameters=true}
   /SetVariables(Particles.SpriteSize=&Vector2DFromFloat(Value=&Multiply_Float(
      A=$Particles.SizeScaledByBoundsFloat
      B=&FloatFromCurve(
       FloatCurve=DataInterface() {propertyValues {Curve {keys [
           {arriveTangent=2.4048, interpMode=RCIM_Cubic, leaveTangent=2.4048, tangentMode=RCTM_User, time=0, value=0.1995}
           {arriveTangent=0.6418, interpMode=RCIM_Cubic, leaveTangent=0.6418, time=0.1363, value=0.4709}
           <at 0.9899 0.8348>
          ]}}}
       Scale Curve=&RandomRangeFloat(Minimum=0.8, Maximum=1.3)
      )
     ))) {isSetParameters=true, module=SetVariables_#4}
   /SetVariables(Particles.MeshColor:LinearColor=&MultiplyLinearColor(Linear Color=$Particles.Color, Scale Factor=&MakeVector4_FromFloat(Float=1))) {isSetParameters=true}
   &self:JanusLightCalculations(JanusLightOffsetToCamera=100, SpawnLightsFraction=4) {module=JanusLightCalculations}
   /SetVariables(Particles.SpriteSizeJanus:Vector2f=&Multiply_Vector2(A=$Particles.SpriteSize, B=&Vector2DFromFloat(Value=$User.JanusSpriteSizeScale))) {isSetParameters=true}
   /SetVariables(Particles.JanusLightIntensity:NiagaraFloat=&FloatFromCurve(FloatCurve=DataInterface() {propertyValues {Curve {keys [<at 0 0 RCIM_Cubic>,<at 0.52 1 RCIM_Cubic>,<at 1 0 RCIM_Cubic>]}}}, Scale Curve=2)) {isSetParameters=true}
  ]
  rendererClasses [
   &NiagaraMeshRendererProperties
   &NiagaraSpriteRendererProperties
   &NiagaraComponentRendererProperties
  ]
  renderers [
   {rendererClass=&NiagaraMeshRendererProperties, rendererIndex=0}
   {rendererClass=&NiagaraSpriteRendererProperties, rendererIndex=1}
   {rendererClass=&NiagaraComponentRendererProperties, rendererIndex=2}
  ]
 }
 FlameLicks {
  emitterUpdateScript [
   &EmitterState(Life Cycle Mode=::NewEnumerator1, Loop Duration=1, UseLoopDelay=false)
   &SpawnRate(SpawnRate=0, Use Spawn Probability=false) {enabled=false}
   &FN_Curie_Fire_Spawn(SpawnCountScale:NiagaraFloat=$Engine.Emitter.SpawnCountScale)
  ]
  particleSpawnScript [
   &InitializeParticle(
    Color=<rgba 1 1 1 0.1>
    Lifetime=&UniformRangedFloat(Minimum=0.35, Maximum=0.65)
    Module.Write Color=true
    Module.Write Lifetime=true
    Module.Write Mass=true
    Module.Write Material Random=true
    Module.Write Position=false
    Module.Write Scale=false
    Module.Write SpriteRotation=false
    Module.Write SpriteSize=true
    Position=&SimulationPosition()
    Sprite Size=&Vector2DFromFloat(Value=&UniformRangedFloat(Minimum=55, Maximum=75))
   )
   &AddVelocity(Velocity=&UniformRangedVector(Minimum=<-40 -40 150>, Maximum=<40 40 70>))
   &DynamicMaterialParameters(
    Index 0 Param 1=3
    Index 0 Param 2=0.5
    Index 0 Param 3=&UniformRangedFloat(Minimum=1, Maximum=2)
    Index 0 Param 4=&Ceil_Float(Float=&UniformRangedFloat(Minimum=1.3, Maximum=2))
    Index 1 Param 1=&Multiply_Float(A=$Particles.SizeScaledByBoundsFloat, B=2)
    Index 1 Param 2=&UniformRangedFloat(Minimum=1, Maximum=2)
    Index 1 Param 3=0.2
    Index 1 Param 4=0.13
    Index 3 Param 4=&self:SetRadiusBiasByScalability_02()
    Module.Param0WriteEnabled=false
    Module.Param0WriteEnabled1=true
    Module.Param0WriteEnabled2=true
    Module.Param0WriteEnabled3=true
    Module.Param1WriteEnabled=true
    Module.Param1WriteEnabled1=true
    Module.Param1WriteEnabled2=true
    Module.Param1WriteEnabled3=true
    Module.Param2WriteEnabled=true
    Module.Param2WriteEnabled1=true
    Module.Param2WriteEnabled2=true
    Module.Param2WriteEnabled3=true
    Module.Param3WriteEnabled=true
    Module.Param3WriteEnabled1=true
    Module.Param3WriteEnabled2=true
    Module.Param3WriteEnabled3=true
    Write Parameter Index 1=true
    Write Parameter Index 2=true
    Write Parameter Index 3=true
   ) {module=DynamicMaterialParameters001}
   &FN_Curie_Fire_InitializePosition()
   /Niagara/Modules/Spawn/Location/V2/SphereLocation(MaskSpawnGroup=false, Override Seed=false, Surface Only=false, UseNonUniformScale=false, UseOffset=false, UseSphereOrientation=false)
   /SetVariables(Particles.Position=&Subtract_Vector(A=Linked() {linkedVariable {name=Particles.Position, type=NiagaraPosition}}, B=<0 0 50>)) {isSetParameters=true}
  ]
  particleUpdateScript [
   &ParticleState()
   &AccelerationForce(Acceleration=<0 0 50>)
   &SolveForcesAndVelocity(Clamp Velocity=false, Limit Acceleration=false)
   &SubUVAnimation(Number Of Frames=&FloatFromCurve(FloatCurve=DataInterface() {propertyValues {Curve {keys [<at 0 1>,<at 1 0>]}}})) {enabled=false, module=SubUVAnimation001}
   /SetVariables(Particles.SubImageIndex=&FloatFromCurve(
     FloatCurve=DataInterface() {propertyValues {Curve {keys [<at 0 22>,<at 1 48>]}}}
     CurveIndex=&ModuloFloat(
      A=&Multiply_Float(
       A=&Add_Float(A=$Particles.NormalizedAge, B=$Particles.MaterialRandom)
       B=&RandomRangeFloat(Minimum=0.8, Maximum=1.2)
      )
      B=1
     )
    )) {isSetParameters=true}
   &DynamicMaterialParameters(
    Index 0 Param 1=&FloatFromCurve(
     FloatCurve=DataInterface() {propertyValues {Curve {keys [
         {arriveTangent=-7.5742, interpMode=RCIM_Cubic, leaveTangent=-7.5742, tangentMode=RCTM_User, time=0.0915, value=1}
         <at 0.25 1 RCIM_Cubic>
         {arriveTangent=0.2118, interpMode=RCIM_Cubic, leaveTangent=0.2118, tangentMode=RCTM_User, time=0.4, value=1}
        ]}}}
     Scale Curve=1.2
    )
    Index 0 Param 2=1
    Index 0 Param 3=&FloatFromCurve(FloatCurve=DataInterface() {propertyValues {Curve {keys [
         {arriveTangent=-0.0921, interpMode=RCIM_Cubic, time=0, value=0.2}
         {arriveTangent=1.8, interpMode=RCIM_Cubic, leaveTangent=1.8, time=0.53, value=0.4}
         <at 1 2>
        ]}}})
    Index 0 Param 4=1
    Index 1 Param 1=&Multiply_Float(
     A=$Particles.SizeScaledByBoundsFloat
     B=&FloatFromCurve(FloatCurve=DataInterface() {propertyValues {Curve {keys [
          {arriveTangent=3.3396, interpMode=RCIM_Cubic, leaveTangent=3.3396, tangentMode=RCTM_User, time=0, value=0.5}
          {arriveTangent=1.5, interpMode=RCIM_Cubic, leaveTangent=1.5, time=0.2393, value=1.3608}
          <at 1 2>
         ]}}})
    )
    Index 1 Param 2=&UniformRangedFloat(Minimum=0.5, Maximum=4)
    Index 1 Param 4=0.1
    Index 2 Param 4=&FloatFromCurve(
     FloatCurve=DataInterface() {propertyValues {Curve {keys [
         <at 0.0177 0.0082 RCIM_Cubic>
         {arriveTangent=-0.0064, interpMode=RCIM_Cubic, leaveTangent=-0.0064, time=0.3544, value=0.8348}
         <at 1.0011 0.002>
        ]}}}
     Scale Curve=&MakeFloatFromLinearColor(LinearColor=$User.LightColor, Channel:ENiagaraLinearColor_Channels=::NewEnumerator3)
    )
    Module.Param0WriteEnabled=true
    Module.Param0WriteEnabled1=true
    Module.Param0WriteEnabled2=true
    Module.Param1WriteEnabled=false
    Module.Param1WriteEnabled1=false
    Module.Param1WriteEnabled2=true
    Module.Param2WriteEnabled=true
    Module.Param2WriteEnabled1=false
    Module.Param2WriteEnabled2=true
    Module.Param3WriteEnabled=false
    Module.Param3WriteEnabled1=true
    Module.Param3WriteEnabled2=false
    Write Parameter Index 1=true
    Write Parameter Index 2=true
   ) {module=DynamicMaterialParameters002}
   &ScaleColor(
    Scale Mode=::NewEnumerator1
    Scale RGBA=&Vector4FromCurve(Vector4Curve=DataInterface() {propertyValues {
       AlphaCurve {keys [<at 0.0024 0>,<at 0.35 1>,<at 0.67 0.67>,<at 0.9977 -0.0026>]}
       BlueCurve {keys [<at 0.3937 1.5>,<at 1 0>]}
       GreenCurve {keys [<at 0.3937 4.2958>,<at 1 0.981>]}
       RedCurve {keys [<at 0.3937 30>,<at 1 10>]}
      }})
    ScaleRGBA=true
   ) {module=ScaleColor001}
   /SetVariables(Particles.SpriteSizeJanus:Vector2f=&Multiply_Vector2(A=$Particles.SpriteSize, B=&Vector2DFromFloat(Value=$User.JanusSpriteSizeScale))) {isSetParameters=true}
  ]
  rendererClasses [&NiagaraSpriteRendererProperties]
  renderers [
   {rendererClass=&NiagaraSpriteRendererProperties}
   {rendererClass=&NiagaraSpriteRendererProperties, rendererIndex=1}
  ]
 }
 Ignition {
  emitterUpdateScript [
   &EmitterState(Life Cycle Mode=::NewEnumerator1, Loop Duration=1, UseLoopDelay=false)
   &SpawnRate(SpawnRate=20, Use Spawn Probability=false) {enabled=false}
   &FN_Curie_Fire_Spawn(UseIgnition=true)
  ]
  particleSpawnScript [
   &InitializeParticle(
    Color=<rgba 12 1.7183 0.6 0.5>
    Lifetime=2
    Module.Write Color=true
    Module.Write Lifetime=true
    Module.Write Mass=true
    Module.Write Material Random=true
    Module.Write Position=false
    Module.Write Scale=false
    Module.Write SpriteRotation=true
    Module.Write SpriteSize=true
    Position=&SimulationPosition()
    Sprite Rotation=&UniformRangedFloat(Minimum=0, Maximum=360)
    Sprite Size=&Vector2DFromFloat(Value=150)
   )
   &FN_Curie_Fire_InitializeSize(ParticleSizeMax=450, ParticleSizeMin=400, SphericalRadiusRangeInputHigh=900, UseIgnition=true)
   &FN_Curie_Fire_InitializePosition(UseIgnition=true)
   &FN_Curie_Fire_InitializeAcceleration(UseIgnition=true)
   &FN_Curie_Fire_InitializeBoundsRadius(UseIgnition=true)
   &FN_Curie_Fire_InitializeSubUV(RandomSubUVStartingFrameMax:NiagaraFloat=$User.NumberOfSubUVFrames)
   &DynamicMaterialParameters(
    Index 0 Param 1=3
    Index 0 Param 2=0.5
    Index 0 Param 3=&UniformRangedFloat(Minimum=1, Maximum=2)
    Index 0 Param 4=&Ceil_Float(Float=&UniformRangedFloat(Minimum=1.3, Maximum=2))
    Index 1 Param 1=&Multiply_Float(A=$Particles.SizeScaledByBoundsFloat, B=1.9)
    Index 1 Param 2=&UniformRangedFloat(Minimum=0, Maximum=0)
    Index 1 Param 4=0.2
    Index 2 Param 1=&MakeFloatFromVector(Vector=$Particles.BoundsOrigin, Channel=::NewEnumerator0)
    Index 2 Param 2=&MakeFloatFromVector(Vector=$Particles.BoundsOrigin, Channel=::NewEnumerator1)
    Index 2 Param 3=&MakeFloatFromVector(Vector=$Particles.BoundsOrigin, Channel=::NewEnumerator2)
    Index 2 Param 4=&Multiply_Float(A=$Particles.SphericalRadius, B=1)
    Index 3 Param 1=1
    Index 3 Param 4=&self:SetRadiusBiasByScalability()
    Module.Param0WriteEnabled=false
    Module.Param0WriteEnabled1=true
    Module.Param0WriteEnabled2=true
    Module.Param0WriteEnabled3=true
    Module.Param1WriteEnabled=true
    Module.Param1WriteEnabled1=true
    Module.Param1WriteEnabled2=true
    Module.Param1WriteEnabled3=true
    Module.Param2WriteEnabled=true
    Module.Param2WriteEnabled1=true
    Module.Param2WriteEnabled2=true
    Module.Param2WriteEnabled3=true
    Module.Param3WriteEnabled=true
    Module.Param3WriteEnabled1=true
    Module.Param3WriteEnabled2=true
    Module.Param3WriteEnabled3=true
    Write Parameter Index 1=true
    Write Parameter Index 2=true
    Write Parameter Index 3=true
   )
   &AddVelocity(Velocity=<0 0 125>) {enabled=false}
  ]
  particleUpdateScript [
   &DynamicMaterialParameters(
    Index 0 Param 1=&FloatFromCurve(FloatCurve=DataInterface() {propertyValues {Curve {keys [
         {arriveTangent=-5.7001, interpMode=RCIM_Cubic, leaveTangent=-5.7001, tangentMode=RCTM_User, time=0, value=1}
         <at 0.15 0 RCIM_Cubic>
         {arriveTangent=5.2486, interpMode=RCIM_Cubic, leaveTangent=5.2486, tangentMode=RCTM_User, time=0.35, value=1}
        ]}}})
    Index 0 Param 2=1
    Index 0 Param 3=&FloatFromCurve(FloatCurve=DataInterface() {propertyValues {Curve {keys [{arriveTangent=-0.0921, interpMode=RCIM_Cubic, time=0, value=1}, <at 1 3>]}}})
    Index 0 Param 4=1
    Index 1 Param 1=&Multiply_Float(
     A=$Particles.SizeScaledByBoundsFloat
     B=&FloatFromCurve(FloatCurve=DataInterface() {propertyValues {Curve {keys [<at 0 0.5 RCIM_Cubic>, {arriveTangent=1.5, interpMode=RCIM_Cubic, leaveTangent=1.5, time=0.24, value=1.36}, <at 1 2>]}}})
    )
    Index 1 Param 4=0.2
    Module.Param0WriteEnabled=true
    Module.Param0WriteEnabled1=true
    Module.Param1WriteEnabled=false
    Module.Param1WriteEnabled1=true
    Module.Param2WriteEnabled=false
    Module.Param2WriteEnabled1=true
    Module.Param3WriteEnabled=false
    Module.Param3WriteEnabled1=true
    Write Parameter Index 1=true
   ) {module=DynamicMaterialParameters001}
   &ParticleState()
   &Drag(Drag=0.5, Use Linear Drag=true, Use Rotational Drag=false) {enabled=false}
   &SolveForcesAndVelocity(Clamp Velocity=false, Limit Acceleration=false)
   &CameraOffset(Camera Offset Amount=35)
   &SubUVAnimation(Number Of Frames=&FloatFromCurve(FloatCurve=DataInterface() {propertyValues {Curve {keys [<at 0 0>,<at 1 200>]}}}))
   /SetVariables(Particles.SubImageIndex=&self:AddRandomFloatToSubimageFrame_2(RandomSubUVStartingFrame:NiagaraFloat=$Particles.RandomSubUVStartingFrame)) {isSetParameters=true}
   /SetVariables(Particles.Scale=&ScaleVectorByCurve(
     InitialValue:Vector3f=&Multiply_VectorByFloat(Vector=&VectorFromFloat(Value=$Particles.SizeScaledByBoundsFloat), Float=1)
     ScaleCurve:NiagaraDataInterfaceVectorCurve=DataInterface() {propertyValues {
       XCurve {keys [<at 0 0.4 RCIM_Cubic>, {arriveTangent=0.6, interpMode=RCIM_Cubic, leaveTangent=0.6, time=0.2, value=0.7}, <at 1 1 RCIM_Cubic>]}
       YCurve {keys [<at 0 0.4 RCIM_Cubic>, {arriveTangent=0.6, interpMode=RCIM_Cubic, leaveTangent=0.6, time=0.2, value=0.7}, <at 1 1 RCIM_Cubic>]}
       ZCurve {keys [<at 0 0.4 RCIM_Cubic>, {arriveTangent=0.6, interpMode=RCIM_Cubic, leaveTangent=0.6, time=0.2, value=0.7}, <at 1 1 RCIM_Cubic>]}
      }}
     CurveIndex:NiagaraFloat=$Particles.NormalizedAge
    )) {isSetParameters=true}
   /SetVariables(Particles.MeshColor:LinearColor=&MultiplyLinearColor(Linear Color=$Particles.Color, Scale Factor=&MakeVector4_FromFloat(Float=0.75))) {isSetParameters=true}
  ]
  rendererClasses [&NiagaraMeshRendererProperties, &NiagaraSpriteRendererProperties]
  renderers [
   {rendererClass=&NiagaraMeshRendererProperties, rendererIndex=0}
   {rendererClass=&NiagaraSpriteRendererProperties, rendererIndex=1}
  ]
 }
}
summary {
 systemName=NS_Curie_Fire
 emitters [
  {emitterName=Embers, rendererClasses [&NiagaraSpriteRendererProperties]}
  {emitterName=Ignition, rendererClasses [&NiagaraMeshRendererProperties, &NiagaraSpriteRendererProperties]}
  {emitterName=EmbersAOE, rendererClasses [&NiagaraSpriteRendererProperties]}
  {
   emitterName=Fire
   rendererClasses [
    &NiagaraMeshRendererProperties
    &NiagaraSpriteRendererProperties
    &NiagaraComponentRendererProperties
   ]
  }
  {emitterName=FlameLicks, rendererClasses [&NiagaraSpriteRendererProperties]}
  {
   emitterName=FireAOE
   rendererClasses [
    &NiagaraMeshRendererProperties
    &NiagaraSpriteRendererProperties
    &NiagaraComponentRendererProperties
   ]
  }
 ]
 userVariables [
  User.CurieFire {
   type=FortCurieFireNiagaraDataInterface
   description=""
   defaultValue {
    struct=NiagaraExt_VariableValue_DataInterface
    value {
     dataInterface=&self:FortCurieFireNiagaraDataInterface_0
     dataInterfaceClass=/Script/FortniteGame.FortCurieFireNiagaraDataInterface
    }
   }
  }
  User.JanusSpriteSizeScale {type=NiagaraFloat, defaultValue=2.1, description=""}
  User.LightColor {type=LinearColor, description="", defaultValue {struct=LinearColor, value=<rgba 1.3 0.3555 0.099 1>}}
  User.LightScalability {type=NiagaraInt32, defaultValue=4}
  User.NumberOfSubUVFrames {type=NiagaraFloat, defaultValue=63, description=""}
 ]
}
