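# Niagara body effect for Curie wildlife that puts smoke, flames and glow on a skeletal mesh.
# Particles pick one of 22 bones by particle ID; Smoke and lit OnFireFlames re-sample it each update and drift upward.
# SmokeTrail spawns by distance moved, sampling bones only at spawn.
# GlowPlanes bursts one particle per bone, passing on-fire, local-player and transition timing user parameters to its material, while another scales flame spawn rate.
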NiagaraSystem /Game/Curie/Effects/NS_Curie_Wildlife_BodySmoke
dependencies {
 usedDataInterfaces [&NiagaraDataInterfaceVectorCurve, &NiagaraDataInterfaceCurve, &NiagaraDataInterfaceColorCurve]
 usedDynamicInputs []
 usedModules [&SystemState]
}
emitters {
 GlowPlanes {
  emitterUpdateScript [
   &EmitterState(
    Inactive Response=::NewEnumerator1
    Life Cycle Mode=::NewEnumerator1
    Loop Behavior=::NewEnumerator1
    Loop Duration Mode=::NewEnumerator1
    UseLoopDelay=false
   )
   &SpawnRate(SpawnRate=0, Use Spawn Probability=false)
   &SpawnBurst_Instantaneous(Spawn Count=22, Use Spawn Probability=false)
  ]
  particleSpawnScript [/Niagara/Modules/Spawn/Initialization/InitializeParticle(
    Color=<rgba 0.1256 1 0 1>
    Lifetime=0
    Mesh Scale=&VectorFromFloat(Value=25)
    Module.Write Color=true
    Module.Write Lifetime=true
    Module.Write Mass=true
    Module.Write Material Random=true
    Module.Write Position=false
    Module.Write Scale=true
    Module.Write SpriteRotation=false
    Module.Write SpriteSize=true
    Position=&SimulationPosition()
    Sprite Size=<1 1>
   )]
  particleUpdateScript [
   &DynamicMaterialParameters(
    Index 0 Param 1=$User.OnFire
    Index 0 Param 2=$User.IsLocalPlayer
    Index 0 Param 3=1
    Index 1 Param 1=$User.TimestampA
    Index 1 Param 2=$User.TransitionLengthIn
    Index 1 Param 3=$User.TimestampB
    Index 1 Param 4=$User.TransitionLengthOut
    Module.Param0WriteEnabled=true
    Module.Param0WriteEnabled1=true
    Module.Param1WriteEnabled=true
    Module.Param1WriteEnabled1=true
    Module.Param2WriteEnabled=true
    Module.Param2WriteEnabled1=true
    Module.Param3WriteEnabled=false
    Module.Param3WriteEnabled1=true
    Write Parameter Index 1=true
   )
   &ParticleState(Kill Particles When Lifetime Has Elapsed=false)
   &SolveForcesAndVelocity(Clamp Velocity=false, Limit Acceleration=false)
   &SkeletalMeshLocation(
    Bone / Socket Index=&ModuloInt(A=/Niagara/DynamicInputs/Execution/ReturnParticleID(), B=22)
    Bone / Socket Sampling Mode=::NewEnumerator5
    Inherit Velocity=false
    MaskSpawnGroup=false
    Mesh Sampling Type=::NewEnumerator6
    Skeletal Mesh=$User.SkelMesh
   )
   &CameraOffset(Camera Offset Amount=30)
  ]
  rendererClasses [&NiagaraMeshRendererProperties, &NiagaraSpriteRendererProperties]
  renderers [
   {rendererClass=&NiagaraMeshRendererProperties, rendererIndex=0}
   {rendererClass=&NiagaraSpriteRendererProperties, rendererIndex=1}
  ]
 }
 OnFireFlames {
  emitterUpdateScript [
   &EmitterState(Inactive Response=::NewEnumerator1, Life Cycle Mode=::NewEnumerator1, Loop Duration=1, UseLoopDelay=false)
   &SpawnRate(SpawnRate=&Multiply_Float(A=60, B=$User.OnFireFlamesSpawnRateScale), Use Spawn Probability=false)
  ]
  particleSpawnScript [
   &InitializeParticle(
    Color=<rgba 40 9.2708 3.125 1>
    Lifetime=&UniformRangedFloat(Minimum=0.4, Maximum=0.7)
    Mesh Scale=<50 50 50>
    Module.Write Color=true
    Module.Write Lifetime=true
    Module.Write Mass=true
    Module.Write Material Random=true
    Module.Write Position=false
    Module.Write Scale=false
    Module.Write SpriteRotation=true
    Module.Write SpriteSize=true
    Position=&SimulationPosition()
    Sprite Rotation=&UniformRangedFloat(Minimum=0, Maximum=360)
    Sprite Size=&Vector2DFromFloat(Value=&UniformRangedFloat(Minimum=20, Maximum=40))
   )
   /Niagara/Modules/Spawn/Location/SphereLocation(MaskSpawnGroup=false, Module.Override Randomness=false, Sphere Radius=1, Surface Only=false)
   /SetVariables(Particles.RandomSphericalOffset:Vector3f=&Multiply_VectorByFloat(Vector=$Output.SphereLocation.SphereNormal, Float=6)) {isSetParameters=true, module=SetVariables_#1}
   &DynamicMaterialParameters(
    Index 0 Param 2=&UniformRangedFloat(Minimum=0.3, Maximum=1)
    Index 0 Param 3=&UniformRangedFloat(Minimum=0.5, Maximum=3)
    Index 0 Param 4=&Ceil_Float(Float=&UniformRangedFloat(Minimum=1, Maximum=2))
    Index 3 Param 1=1
    Index 3 Param 2=1
    Module.Param0WriteEnabled=true
    Module.Param0WriteEnabled3=false
    Module.Param1WriteEnabled=true
    Module.Param1WriteEnabled3=false
    Module.Param2WriteEnabled=true
    Module.Param2WriteEnabled3=false
    Module.Param3WriteEnabled=true
    Module.Param3WriteEnabled3=false
    Write Parameter Index 3=true
   )
  ]
  particleUpdateScript [
   &SampleSkeletalMeshSkeleton(
    Mesh:NiagaraDataInterfaceSkeletalMesh=$User.SkelMesh
    Specific Bone Index:NiagaraInt32=&ModuloInt(A=&ReturnParticleID(), B=22)
   )
   /Niagara/Modules/Spawn/Location/SkeletalMeshSkeletonLocation(
    Module.Inherit Velocity=false
    Module.Use Sampled Bone as Ribbon ID=true
    Sampled Position Attribute:Vector3f=&Add_Vector(A=$Particles.SampleSkeletalMeshSkeleton.SampledPosition, B=$Particles.RandomSphericalOffset)
    Sampled Position Offset:Vector3f=&VectorFromCurve(VectorCurve=DataInterface() {propertyValues {
       XCurve {keys [<at 0 0>,<at 1 0>]}
       YCurve {keys [<at 0 0>,<at 1 0>]}
       ZCurve {keys [<at 0 0 RCIM_Cubic>, {arriveTangent=44.294, interpMode=RCIM_Cubic, leaveTangent=44.294, tangentMode=RCTM_User, time=1, value=20}]}
      }})
   )
   &ParticleState()
   /Niagara/Modules/Update/SubUV/SubUVAnimation(Number Of Frames=&FloatFromCurve(FloatCurve=DataInterface() {propertyValues {Curve {keys [<at 0 0>,<at 1.0068 55>]}}}))
   &DynamicMaterialParameters(
    Index 0 Param 1=&FloatFromCurve(FloatCurve=DataInterface() {propertyValues {Curve {keys [<at 0 0>,<at 0.85 1.2>]}}})
    Index 0 Param 3=&FloatFromCurve(FloatCurve=DataInterface() {propertyValues {Curve {keys [
         <at 0.0123 1.3769 RCIM_Cubic>
         {arriveTangent=-0.0921, interpMode=RCIM_Cubic, leaveTangent=-0.0921, time=0.367, value=1.1881}
         <at 0.9928 1.2866>
        ]}}})
    Index 0 Param 4=0.2
    Index 2 Param 1=&MakeFloatFromLinearColor(LinearColor=?, Channel:ENiagaraLinearColor_Channels=::NewEnumerator0)
    Index 2 Param 2=&MakeFloatFromLinearColor(LinearColor=?, Channel:ENiagaraLinearColor_Channels=::NewEnumerator1)
    Index 2 Param 3=&MakeFloatFromLinearColor(LinearColor=?, Channel:ENiagaraLinearColor_Channels=::NewEnumerator2)
    Index 2 Param 4=&FloatFromCurve(
     FloatCurve=DataInterface() {propertyValues {Curve {keys [
         <at 0.0177 0.0082 RCIM_Cubic>
         {arriveTangent=-0.0064, interpMode=RCIM_Cubic, leaveTangent=-0.0064, time=0.3544, value=0.8348}
         <at 1.0011 0.002>
        ]}}}
     Scale Curve=&MakeFloatFromLinearColor(LinearColor=?, Channel:ENiagaraLinearColor_Channels=::NewEnumerator3)
    )
    Module.Param0WriteEnabled=true
    Module.Param0WriteEnabled2=true
    Module.Param1WriteEnabled=false
    Module.Param1WriteEnabled2=true
    Module.Param2WriteEnabled=false
    Module.Param2WriteEnabled2=true
    Module.Param3WriteEnabled=false
    Module.Param3WriteEnabled2=true
    Write Parameter Index 2=true
   ) {module=DynamicMaterialParameters001}
   &SolveForcesAndVelocity(Clamp Velocity=false, Limit Acceleration=false)
   &ScaleColor(
    Scale Alpha=&FloatFromCurve(FloatCurve=DataInterface() {propertyValues {Curve {keys [
         <at 0 0 RCIM_Cubic>
         {arriveTangent=0.9268, interpMode=RCIM_Cubic, leaveTangent=0.9268, time=0.27, value=1}
         {arriveTangent=-1.365, interpMode=RCIM_Cubic, leaveTangent=-1.365, time=0.7207, value=0.668}
         <at 1.0011 0.002>
        ]}}})
    Scale RGB=&VectorFromCurve(VectorCurve=DataInterface() {propertyValues {
       XCurve {keys [<at 0 0>, <at 0.2 1.0067>, {arriveTangent=-1.2584, interpMode=RCIM_Cubic, leaveTangent=-1.2584, time=0.8, value=0.9431}, <at 1 0>]}
       YCurve {keys [<at 0 0>, <at 0.2 1.0067>, {arriveTangent=-1.2584, interpMode=RCIM_Cubic, leaveTangent=-1.2584, time=0.8, value=0.4309}, <at 1 0>]}
       ZCurve {keys [<at 0 0>, <at 0.2 1.0067>, {arriveTangent=-1.2584, interpMode=RCIM_Cubic, leaveTangent=-1.2584, time=0.8, value=0}, <at 1 0>]}
      }})
    ScaleA=true
    ScaleRGB=true
   )
   &CameraOffset(Camera Offset Amount=4)
   &ScaleSpriteSize(Scale Factor=&Vector2DFromFloat(Value=&FloatFromCurve(FloatCurve=DataInterface() {propertyValues {Curve {keys [{arriveTangent=0.9765, interpMode=RCIM_Cubic, leaveTangent=0.9765, tangentMode=RCTM_User, time=0, value=0.6}, <at 1 1>]}}})))
   &/LightAttributeSetter(
    LightColor:LinearColor=&ScaleLinearColorByCurve(
     InitialValue=<rgba 1.5 0.3129 0.0755 1>
     ScaleCurve=DataInterface() {propertyValues {
       AlphaCurve {keys []}
       BlueCurve {keys [<at 0.0023 0>,<at 0.2206 12>,<at 1.033 0>]}
       GreenCurve {keys [<at 0.0023 0>,<at 0.2206 12>,<at 1.033 0>]}
       RedCurve {keys [<at 0.0023 0>,<at 0.2206 12>,<at 1.033 0>]}
      }}
     CurveIndex=$Particles.NormalizedAge
    )
    LightRadius=45
   )
  ]
  rendererClasses [&NiagaraLightRendererProperties, &NiagaraSpriteRendererProperties]
  renderers [
   {rendererClass=&NiagaraLightRendererProperties, rendererIndex=0}
   {rendererClass=&NiagaraSpriteRendererProperties, rendererIndex=1}
  ]
 }
 Smoke {
  emitterUpdateScript [
   &EmitterState(Inactive Response=::NewEnumerator1, Life Cycle Mode=::NewEnumerator1, Loop Duration=1, UseLoopDelay=false)
   &SpawnRate(SpawnRate=15, Use Spawn Probability=false)
  ]
  particleSpawnScript [
   &InitializeParticle(
    Color=$User.SmokeColor
    Lifetime=&UniformRangedFloat(Minimum=0.5, Maximum=0.9)
    Mesh Scale=<50 50 50>
    Module.Write Color=true
    Module.Write Lifetime=true
    Module.Write Mass=true
    Module.Write Material Random=true
    Module.Write Position=false
    Module.Write Scale=false
    Module.Write SpriteRotation=true
    Module.Write SpriteSize=true
    Position=&SimulationPosition()
    Sprite Rotation=&UniformRangedFloat(Minimum=0, Maximum=360)
    Sprite Size=&Vector2DFromFloat(Value=&UniformRangedFloat(Minimum=90, Maximum=125))
   )
   &SphereLocation(MaskSpawnGroup=false, Module.Override Randomness=false, Sphere Radius=1, Surface Only=false)
   /SetVariables(Particles.RandomSphericalOffset:Vector3f=&Multiply_VectorByFloat(Vector=$Output.SphereLocation.SphereNormal, Float=12)) {isSetParameters=true, module=SetVariables_#1}
  ]
  particleUpdateScript [
   &SampleSkeletalMeshSkeleton(
    Mesh:NiagaraDataInterfaceSkeletalMesh=$User.SkelMesh
    Specific Bone Index:NiagaraInt32=&ModuloInt(A=&ReturnParticleID(), B=22)
   )
   &SkeletalMeshSkeletonLocation(
    Module.Inherit Velocity=false
    Module.Use Sampled Bone as Ribbon ID=true
    Sampled Position Attribute:Vector3f=&Add_Vector(A=$Particles.SampleSkeletalMeshSkeleton.SampledPosition, B=$Particles.RandomSphericalOffset)
    Sampled Position Offset:Vector3f=&VectorFromCurve(VectorCurve=DataInterface() {propertyValues {
       XCurve {keys [<at 0 0>,<at 1 0>]}
       YCurve {keys [<at 0 0>,<at 1 0>]}
       ZCurve {keys [
         {arriveTangent=17.3803, interpMode=RCIM_Cubic, leaveTangent=17.3802, tangentMode=RCTM_User, time=0, value=0}
         {arriveTangent=44.294, interpMode=RCIM_Cubic, leaveTangent=44.294, tangentMode=RCTM_User, time=1, value=30}
        ]}
      }})
   )
   &ParticleState()
   &SubUVAnimation(Number Of Frames=&FloatFromCurve(FloatCurve=DataInterface() {propertyValues {Curve {keys [<at 0 0>,<at 1.0068 30>]}}}))
   &SolveForcesAndVelocity(Clamp Velocity=false, Limit Acceleration=false)
   &ScaleColor(
    Scale Alpha=&FloatFromCurve(FloatCurve=DataInterface() {propertyValues {Curve {keys [<at 0 0 RCIM_Cubic>, {arriveTangent=0.002, interpMode=RCIM_Cubic, leaveTangent=0.002, time=0.27, value=1}, <at 1.0011 0.002>]}}})
    Scale RGB=&VectorFromCurve(VectorCurve=DataInterface() {propertyValues {XCurve {keys [<at 0 1>]}, YCurve {keys [<at 0 1>]}, ZCurve {keys [<at 0 1>]}}})
    ScaleA=true
    ScaleRGB=false
   )
   &CameraOffset(Camera Offset Amount=4)
   &ScaleSpriteSize(Scale Factor=&Vector2DFromFloat(Value=&FloatFromCurve(FloatCurve=DataInterface() {propertyValues {Curve {keys [
          {arriveTangent=0.6031, interpMode=RCIM_Cubic, leaveTangent=0.6031, tangentMode=RCTM_User, time=0, value=0.2}
          {arriveTangent=0.2366, interpMode=RCIM_Cubic, leaveTangent=0.2366, tangentMode=RCTM_User, time=1, value=1}
         ]}}})))
   &DynamicMaterialParameters(
    Index 0 Param 1=&FloatFromCurve(FloatCurve=DataInterface() {propertyValues {Curve {keys [
         {arriveTangent=5.8249, interpMode=RCIM_Cubic, leaveTangent=5.8249, tangentMode=RCTM_User, time=0, value=-1}
         <at 0.1868 0.2629 RCIM_Cubic>
         <at 1.0011 -1>
        ]}}})
    Module.Param0WriteEnabled=true
    Module.Param1WriteEnabled=true
    Module.Param2WriteEnabled=true
    Module.Param3WriteEnabled=true
   )
  ]
  rendererClasses [&NiagaraSpriteRendererProperties]
  renderers [{rendererClass=&NiagaraSpriteRendererProperties}]
 }
 SmokeTrail {
  emitterUpdateScript [
   &EmitterState(Inactive Response=::NewEnumerator1, Life Cycle Mode=::NewEnumerator1, Loop Duration=1, UseLoopDelay=false)
   &SpawnPerUnit(Module.Use Max Movement Threshold=true, Module.Use Movement Tolerance=false, Spawn Per Unit=20, Use Spawn Probability=false)
   &SpawnRate(SpawnRate=20, Use Spawn Probability=false) {enabled=false}
  ]
  particleSpawnScript [
   &InitializeParticle(
    Color=$User.SmokeColor
    Lifetime=&UniformRangedFloat(Minimum=0.4, Maximum=0.5)
    Mesh Scale=<50 50 50>
    Module.Write Color=true
    Module.Write Lifetime=true
    Module.Write Mass=true
    Module.Write Material Random=true
    Module.Write Position=false
    Module.Write Scale=false
    Module.Write SpriteRotation=true
    Module.Write SpriteSize=true
    Position=&SimulationPosition()
    Sprite Rotation=&UniformRangedFloat(Minimum=0, Maximum=360)
    Sprite Size=&Vector2DFromFloat(Value=&UniformRangedFloat(Minimum=120, Maximum=160))
   )
   &SphereLocation(MaskSpawnGroup=false, Module.Override Randomness=false, Sphere Radius=1, Surface Only=false)
   &SampleSkeletalMeshSkeleton(
    Mesh:NiagaraDataInterfaceSkeletalMesh=$User.SkelMesh
    Specific Bone Index:NiagaraInt32=&ModuloInt(A=&ReturnParticleID(), B=22)
   )
   &SkeletalMeshSkeletonLocation(Module.Inherit Velocity=false, Module.Use Sampled Bone as Ribbon ID=true)
   /SetVariables(Particles.RandomSphericalOffset:Vector3f=&Multiply_VectorByFloat(Vector=$Output.SphereLocation.SphereNormal, Float=12)) {isSetParameters=true, module=SetVariables_#1}
  ]
  particleUpdateScript [
   &ParticleState()
   &SubUVAnimation(Number Of Frames=&FloatFromCurve(FloatCurve=DataInterface() {propertyValues {Curve {keys [<at 0 0>,<at 1.0068 30>]}}}))
   &InheritVelocity(Module.Limit Speed=true, Velocity Limit=20, Velocity Scale=&VectorFromFloat(Value=0.5))
   &Drag(Use Linear Drag=true, Use Rotational Drag=false)
   &SolveForcesAndVelocity(Clamp Velocity=false, Limit Acceleration=false)
   &ScaleColor(
    Scale Alpha=&FloatFromCurve(FloatCurve=DataInterface() {propertyValues {Curve {keys [<at 0 0 RCIM_Cubic>, {arriveTangent=0.002, interpMode=RCIM_Cubic, leaveTangent=0.002, time=0.27, value=1}, <at 1.0011 0.002>]}}})
    Scale RGB=&VectorFromCurve(VectorCurve=DataInterface() {propertyValues {XCurve {keys [<at 0 1>]}, YCurve {keys [<at 0 1>]}, ZCurve {keys [<at 0 1>]}}})
    ScaleA=true
    ScaleRGB=false
   )
   &CameraOffset(Camera Offset Amount=4)
   &ScaleSpriteSize(Scale Factor=&Vector2DFromFloat(Value=&FloatFromCurve(FloatCurve=DataInterface() {propertyValues {Curve {keys [
          {arriveTangent=0.6031, interpMode=RCIM_Cubic, leaveTangent=0.6031, tangentMode=RCTM_User, time=0, value=0.2}
          {arriveTangent=0.2366, interpMode=RCIM_Cubic, leaveTangent=0.2366, tangentMode=RCTM_User, time=1, value=1}
         ]}}})))
   &DynamicMaterialParameters(
    Index 0 Param 1=&FloatFromCurve(FloatCurve=DataInterface() {propertyValues {Curve {keys [<at 0 -1>,<at 0.2352 0.5795>,<at 1.0011 -1>]}}})
    Module.Param0WriteEnabled=true
    Module.Param1WriteEnabled=true
    Module.Param2WriteEnabled=true
    Module.Param3WriteEnabled=true
   )
  ]
  rendererClasses [&NiagaraSpriteRendererProperties]
  renderers [{rendererClass=&NiagaraSpriteRendererProperties}]
 }
}
summary {
 systemName=NS_Curie_Wildlife_BodySmoke
 emitters [
  {emitterName=Smoke, rendererClasses [&NiagaraSpriteRendererProperties]}
  {emitterName=SmokeTrail, rendererClasses [&NiagaraSpriteRendererProperties]}
  {emitterName=GlowPlanes, rendererClasses [&NiagaraMeshRendererProperties, &NiagaraSpriteRendererProperties]}
  {emitterName=OnFireFlames, rendererClasses [&NiagaraLightRendererProperties, &NiagaraSpriteRendererProperties]}
 ]
 userVariables [
  User.FlameSizeMax {type=NiagaraFloat, defaultValue=0, description=""}
  User.FlameSizeMin {type=NiagaraFloat, defaultValue=0, description=""}
  User.IsLocalPlayer {type=NiagaraFloat, defaultValue=0, description=""}
  User.OnFire {type=NiagaraFloat, defaultValue=0, description=""}
  User.OnFireFlamesSpawnRateScale {type=NiagaraFloat, defaultValue=1, description=""}
  User.RandomOffset {type=Vector3f, description="", defaultValue {struct=Vector3f, value=<0 0 0>}}
  User.SkelMesh {
   type=NiagaraDataInterfaceSkeletalMesh
   defaultValue {
    struct=NiagaraExt_VariableValue_DataInterface
    value {dataInterface=&self:NiagaraDataInterfaceSkeletalMesh_0}
   }
  }
  User.SmokeColor {type=LinearColor, description="", defaultValue {struct=LinearColor, value=<rgba 0 0 0 1>}}
  User.TimestampA {type=NiagaraFloat, defaultValue=0, description=""}
  User.TimestampB {type=NiagaraFloat, defaultValue=0, description=""}
  User.TransitionLengthIn {type=NiagaraFloat, defaultValue=0, description=""}
  User.TransitionLengthOut {type=NiagaraFloat, defaultValue=0, description=""}
 ]
}
